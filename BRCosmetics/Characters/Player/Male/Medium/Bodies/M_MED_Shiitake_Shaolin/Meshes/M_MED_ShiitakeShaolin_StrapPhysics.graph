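# Physics asset for the dangling straps on the Shiitake Shaolin medium male player body.
# Mirrored left and right chains link a small base capsule to three box segments, with a hinge-like first joint (one swing and twist locked) and swing and twist limits below.
# Capsules cover calves, calf twists and feet, joined by free-rotation constraints.
# A lone right coat-tail box has no constraint and no left counterpart.
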PhysicsAsset /BRCosmetics/Characters/Player/Male/Medium/Bodies/M_MED_Shiitake_Shaolin/Meshes/M_MED_ShiitakeShaolin_StrapPhysics
bodies {
 calf_l [{shapeType=Capsule, length=8.7236, radius=7.8349, center=<-5.1337 -3.0407 -0.0165>, rotation=<rot -14.9971 -90 -177.3722>}]
 calf_r [{shapeType=Capsule, length=8.7236, radius=7.8349, center=<5.1337 3.0407 0.0165>, rotation=<rot -14.9971 -90 2.6278>}]
 calf_twist_02_l [{shapeType=Capsule, length=25.9108, radius=5.136, center=<-5.8416 -0.839 -0.1775>, rotation=<rot 90 0 -167.398>}]
 calf_twist_02_r [{shapeType=Capsule, length=25.9108, radius=5.136, center=<5.8416 0.839 0.1775>, rotation=<rot 90 0 12.602>}]
 dyn_coatTail05_r [{shapeType=Box, extentX=1.3697, extentY=7.9835, extentZ=18.1453, center=<3.7304 -0.0001 0>, rotation=<rot 0 90 0>}]
 dyn_strap_03_Base_l [{shapeType=Capsule, length=0.505, radius=0.505, rotation=<rot 0 -90 0>}]
 dyn_strap_03_Base_r [{shapeType=Capsule, length=0.505, radius=0.505, rotation=<rot 0 -90 0>}]
 dyn_strap_03_l [{shapeType=Box, extentX=1.5123, extentY=15.4533, extentZ=7.0878, center=<-6.5844 0.0002 0>, rotation=<rot 10.3209 89.9999 180>}]
 dyn_strap_03_r [{shapeType=Box, extentX=1.5123, extentY=15.4533, extentZ=7.0878, center=<6.5844 -0.0002 0>, rotation=<rot 10.3209 89.9999 0>}]
 dyn_strap_04_l [{shapeType=Box, extentX=1.4635, extentY=12.3462, extentZ=7.1256, center=<-4.7853 0.0002 0>, rotation=<rot 2.0293 89.9999 -180>}]
 dyn_strap_04_r [{shapeType=Box, extentX=1.4635, extentY=12.3462, extentZ=7.1256, center=<4.7853 -0.0002 0>, rotation=<rot 2.0293 89.9999 0>}]
 dyn_strap_05_l [{shapeType=Box, extentX=1.01, extentY=10.6019, extentZ=6.401, center=<-4.8696 -0.1827 0>, rotation=<rot 2.5665 89.9998 -180>}]
 dyn_strap_05_r [{shapeType=Box, extentX=1.01, extentY=10.6019, extentZ=6.401, center=<4.8696 0.1827 0>, rotation=<rot 2.5665 89.9999 0>}]
 foot_l [{shapeType=Capsule, length=8.2559, radius=5.5298, center=<-0.815 1.8665 0>, rotation=<rot 90 12.9074 -171.5226>}]
 foot_r [{shapeType=Capsule, length=8.2559, radius=5.5298, center=<0.815 -1.8665 0>, rotation=<rot 90 -90 -94.4301>}]
}
constraints [
 {bone1Name=dyn_strap_04_r, bone2Name=dyn_strap_03_r, swing1LimitDegrees=30, swing2LimitDegrees=20, twistLimitDegrees=6}
 {bone1Name=dyn_strap_05_r, bone2Name=dyn_strap_04_r, swing1LimitDegrees=30, swing2LimitDegrees=20, twistLimitDegrees=6}
 {
  bone1Name=dyn_strap_03_r
  bone2Name=dyn_strap_03_Base_r
  swing1LimitDegrees=16
  swing2LimitDegrees=16
  swing2Motion=Locked
  twistLimitDegrees=16
  twistMotion=Locked
 }
 {
  bone1Name=calf_twist_02_r
  bone2Name=calf_r
  swing1LimitDegrees=45
  swing1Motion=Free
  swing2LimitDegrees=45
  swing2Motion=Free
  twistLimitDegrees=45
  twistMotion=Free
 }
 {
  bone1Name=foot_r
  bone2Name=calf_r
  swing1LimitDegrees=45
  swing1Motion=Free
  swing2LimitDegrees=45
  swing2Motion=Free
  twistLimitDegrees=45
  twistMotion=Free
 }
 {
  bone1Name=calf_twist_02_l
  bone2Name=calf_l
  swing1LimitDegrees=45
  swing1Motion=Free
  swing2LimitDegrees=45
  swing2Motion=Free
  twistLimitDegrees=45
  twistMotion=Free
 }
 {
  bone1Name=foot_l
  bone2Name=calf_l
  swing1LimitDegrees=45
  swing1Motion=Free
  swing2LimitDegrees=45
  swing2Motion=Free
  twistLimitDegrees=45
  twistMotion=Free
 }
 {bone1Name=dyn_strap_04_l, bone2Name=dyn_strap_03_l, swing1LimitDegrees=30, swing2LimitDegrees=20, twistLimitDegrees=6}
 {bone1Name=dyn_strap_05_l, bone2Name=dyn_strap_04_l, swing1LimitDegrees=30, swing2LimitDegrees=20, twistLimitDegrees=6}
 {
  bone1Name=dyn_strap_03_l
  bone2Name=dyn_strap_03_Base_l
  swing1LimitDegrees=16
  swing2LimitDegrees=16
  swing2Motion=Locked
  twistLimitDegrees=16
  twistMotion=Locked
 }
]
