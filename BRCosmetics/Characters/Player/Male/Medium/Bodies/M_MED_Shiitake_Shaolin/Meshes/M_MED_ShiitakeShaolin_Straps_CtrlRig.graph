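# Control Rig for the Shiitake Shaolin straps mesh that applies pose-driven corrective offsets to coat bones.
# Thigh rotations relative to the pelvis become Euler angles, whose absolute Z is remapped into clamped weights.
# These weights blend additive local offsets onto the right front coat base and the left front and side bases.
# A left upper-arm twist-roll chain is disconnected from execution and targets a bone absent from the hierarchy.
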ControlRigBlueprint /BRCosmetics/Characters/Player/Male/Medium/Bodies/M_MED_Shiitake_Shaolin/Meshes/M_MED_ShiitakeShaolin_Straps_CtrlRig
graphs [
 RigVMFunctionLibrary {links [], nodes []}
 RigVMModel {nodes [
   n1 = &RigUnit_BeginExecution()
   n2 = &RigVMFunction_MathQuaternionToEuler(RotationOrder=ZXY, Value=&RigUnit_GetRelativeBoneTransform(Bone=thigh_r, Space=pelvis).Transform.Rotation)
   /Script/ControlRig.RigUnit_ModifyBoneTransforms(
    BoneToModify=((Bone="dyn_coatFront01_Base_r",Transform=(Rotation=(X=0,Y=0,Z=0,W=1),Translation=(X=0,Y=-5,Z=0),Scale3D=(X=1,Y=1,Z=1))))
    Mode=AdditiveLocal
    WeightMaximum=1
    WeightMinimum=0
    ExecutePin=@n1.ExecutePin
    Weight=&RigVMFunction_MathFloatRemap(
     SourceMaximum=68
     SourceMinimum=178
     TargetMinimum=0
     bClamp=true
     Value=/Script/RigVM.RigVMFunction_MathFloatAbs(Value=@n2.Result.Z) {is=struct RigVMUnitNode, template="Absolute::Execute(in Value,out Result)"}.Result
    ).Result
    BoneToModify.0.Transform.Rotation=&RigVMFunction_MathQuaternionFromEuler(Euler=(X=0,Y=30,Z=0)).Result
   ) {is=struct RigVMUnitNode}
   n3 = &RigUnit_GetRelativeBoneTransform(Bone=thigh_l, Space=pelvis)
   n4 = &RigVMFunction_MathFloatRemap(
    SourceMaximum=106
    SourceMinimum=20
    TargetMinimum=0
    bClamp=true
    Value=&RigVMFunction_MathFloatAbs(Value=&RigVMFunction_MathQuaternionToEuler(RotationOrder=ZXY, Value=@n3.Transform.Rotation).Result.Z) {template="Absolute::Execute(in Value,out Result)"}.Result
   )
   > &RigUnit_ModifyBoneTransforms(
    BoneToModify=((Bone="dyn_coatFront01_Base_l",Transform=(Rotation=(X=0,Y=0,Z=0,W=1),Translation=(X=-2,Y=12,Z=0),Scale3D=(X=1,Y=1,Z=1))))
    Mode=AdditiveLocal
    WeightMaximum=1
    WeightMinimum=0
    Weight=@n4.Result
    BoneToModify.0.Transform.Rotation=&RigVMFunction_MathQuaternionFromEuler(Euler=(X=-15,Y=0,Z=0)).Result
   )
   > &RigUnit_ModifyBoneTransforms(
    BoneToModify=((Bone="dyn_coatSide01_Base_l",Transform=(Rotation=(X=0,Y=0,Z=0,W=1),Translation=(X=0,Y=5,Z=0),Scale3D=(X=1,Y=1,Z=1))))
    Mode=AdditiveLocal
    WeightMaximum=1
    WeightMinimum=0
    Weight=@n4.Result
   )
   n5 = &RigVMFunction_MathQuaternionToEuler(RotationOrder=XZY, Value=&RigUnit_GetRelativeBoneTransform(Bone=upperarm_l, Space=clavicle_l).Transform.Rotation)
   &RigUnit_ModifyBoneTransforms(
    BoneToModify=((Bone="dyn_upperTwistRoll_l",Transform=(Rotation=(X=0,Y=0,Z=0,W=1),Translation=(X=0,Y=0,Z=0),Scale3D=(X=1,Y=1,Z=1))))
    ExecutePin=()
    Mode=AdditiveLocal
    WeightMaximum=1
    WeightMinimum=0
    Weight=&RigVMFunction_MathFloatRemap(SourceMaximum=114, SourceMinimum=50, TargetMinimum=0, bClamp=true, Value=@n5.Result.X).Result
    BoneToModify.0.Transform.Rotation=&RigVMFunction_MathQuaternionFromEuler(Euler=(X=0,Y=30,Z=0)).Result
   )
   > &RigUnit_ModifyBoneTransforms(
    BoneToModify=((Bone="dyn_upperTwistRoll_l",Transform=(Rotation=(X=0,Y=0,Z=0,W=1),Translation=(X=0,Y=0,Z=0),Scale3D=(X=1,Y=1,Z=1))))
    Mode=AdditiveLocal
    WeightMaximum=1
    WeightMinimum=0
    Weight=&RigVMFunction_MathFloatRemap(SourceMaximum=-5, SourceMinimum=15, TargetMinimum=0, bClamp=true, Value=@n5.Result.Z).Result
    BoneToModify.0.Transform.Rotation=&RigVMFunction_MathQuaternionFromEuler(Euler=(X=0,Y=30,Z=0), RotationOrder=XZY).Result
   )
  ]}
]
hierarchy [
 root {loc=<0 0 0>, rot=<0 0 0 1>}
 attach {parent=root, loc=<0 0 0>, rot=<0 0 0 1>}
 pelvis {parent=root, loc=<0 -0.7071 93.1191>, rot=<0.0127 0.707 -0.0127 -0.707>}
 spine_01 {parent=pelvis, loc=<6.7494 -0.6388 0>, rot=<0 0 0.034 -0.9994>}
 spine_02 {parent=spine_01, loc=<7.2518 -0.035 0>, rot=<0 0 0.0238 -0.9997>}
 spine_03 {parent=spine_02, loc=<6.0104 0 0>, rot=<0 0 0.0149 -0.9999>}
 spine_04 {parent=spine_03, loc=<8.1024 0 0>, rot=<0 0 -0.0132 -0.9999>}
 spine_05 {parent=spine_04, loc=<10.0146 0 0>, rot=<0 0 -0.0015 -1>}
 clavicle_l {parent=spine_05, loc=<4.3575 -0.1466 -2.3223>, rot=<0.0465 -0.7289 -0.009 -0.6829>}
 upperarm_l {parent=clavicle_l, loc=<15.5111 0 0>, rot=<0.0226 -0.3234 -0.0088 -0.9459>}
 lowerarm_l {parent=upperarm_l, loc=<26.6514 0 0>, rot=<0.0391 -0.0099 -0.2453 -0.9686>}
 hand_l {parent=lowerarm_l, loc=<22.3491 0 0>, rot=<-0.7081 -0.0507 0.0975 -0.6975>}
 index_metacarpal_l {parent=hand_l, loc=<1.6535 0.0183 -2.96>, rot=<0.0158 -0.0124 -0.0173 -0.9996>}
 index_01_l {parent=index_metacarpal_l, loc=<8.3008 0.0357 -0.0524>, rot=<-0.0044 0.0485 0.0924 -0.9945>}
 index_02_l {parent=index_01_l, loc=<3.3587 -0.3004 0.0452>, rot=<-0.002 0.0027 0.3225 -0.9466>}
 index_03_l {parent=index_02_l, loc=<2.5578 -0.123 -0.0252>, rot=<-0.002 0.0067 0.2558 -0.9667>}
 middle_metacarpal_l {parent=hand_l, loc=<2.1856 -0.1725 -0.5984>, rot=<0.006 0.0168 -0.0329 -0.9993>}
 middle_01_l {parent=middle_metacarpal_l, loc=<7.3275 -0.2447 0.0129>, rot=<-0.0208 0.0042 0.0858 -0.9961>}
 middle_02_l {parent=middle_01_l, loc=<4.0123 0.1401 0.1175>, rot=<-0.0218 0.0258 0.3512 -0.9357>}
 middle_03_l {parent=middle_02_l, loc=<2.6077 0.0534 0.1373>, rot=<0.0082 0.0188 0.1708 -0.9851>}
 pinky_metacarpal_l {parent=hand_l, loc=<1.7616 -0.9753 3.3814>, rot=<-0.0397 0.0814 -0.069 -0.9935>}
 pinky_01_l {parent=pinky_metacarpal_l, loc=<7.1709 -0.4831 -0.1094>, rot=<-0.0654 -0.0035 0.2242 -0.9723>}
 pinky_02_l {parent=pinky_01_l, loc=<2.8274 0.0893 -0.1043>, rot=<0.0037 0.0011 0.2854 -0.9584>}
 pinky_03_l {parent=pinky_02_l, loc=<2.2046 0.0307 -0.0078>, rot=<-0.0002 0.0098 0.1445 -0.9895>}
 ring_metacarpal_l {parent=hand_l, loc=<2.2074 -0.3625 1.3638>, rot=<-0.0284 0.0722 -0.0458 -0.9959>}
 ring_01_l {parent=ring_metacarpal_l, loc=<7.0295 -0.6197 -0.1446>, rot=<-0.0489 -0.0252 0.1892 -0.9804>}
 ring_02_l {parent=ring_01_l, loc=<3.6174 0.4646 -0.1027>, rot=<0 0 0.2806 -0.9598>}
 ring_03_l {parent=ring_02_l, loc=<2.4152 0.1141 0.0199>, rot=<0 0 0.1189 -0.9929>}
 thumb_01_l {parent=hand_l, loc=<1.3118 -0.3191 -2.459>, rot=<0.6546 -0.4042 0.0664 -0.6354>}
 thumb_02_l {parent=thumb_01_l, loc=<5.9336 -0.0448 0.0401>, rot=<-0.0037 0.0213 0.298 -0.9543>}
 thumb_03_l {parent=thumb_02_l, loc=<2.2223 0.3682 -0.0907>, rot=<0.0118 -0.013 0.0066 -0.9998>}
 weapon_l {parent=hand_l, loc=<-0.1974 -3.4111 0.3811>, rot=<0 0 0 1>}
 lowerarm_twist_01_l {parent=lowerarm_l, loc=<17.8549 0 0>, rot=<0 0 0 1>}
 lowerarm_twist_02_l {parent=lowerarm_l, loc=<9.6199 0 0>, rot=<0 0 0 1>}
 upperarm_twist_01_l {parent=upperarm_l, loc=<0.0041 0 0>, rot=<0 0 0 -1>}
 upperarm_twist_02_l {parent=upperarm_l, loc=<13.7941 0 0>, rot=<0 0 0 -1>}
 clavicle_r {parent=spine_05, loc=<4.3575 -0.1466 2.3223>, rot=<0.7289 0.0465 -0.6829 0.009>}
 upperarm_r {parent=clavicle_r, loc=<-15.5111 0 0>, rot=<0.0226 -0.3234 -0.0088 -0.9459>}
 lowerarm_r {parent=upperarm_r, loc=<-26.6514 0 0>, rot=<0.0391 -0.0099 -0.2453 -0.9686>}
 hand_r {parent=lowerarm_r, loc=<-22.3492 0 0>, rot=<-0.7081 -0.0507 0.0975 -0.6975>}
 index_metacarpal_r {parent=hand_r, loc=<-1.6534 -0.0181 2.9601>, rot=<0.0158 -0.0124 -0.0173 -0.9996>}
 index_01_r {parent=index_metacarpal_r, loc=<-8.3008 -0.0357 0.0524>, rot=<-0.0044 0.0485 0.0924 -0.9945>}
 index_02_r {parent=index_01_r, loc=<-3.3589 0.3003 -0.0453>, rot=<-0.002 0.0027 0.3225 -0.9466>}
 index_03_r {parent=index_02_r, loc=<-2.5578 0.1231 0.0251>, rot=<-0.002 0.0067 0.2558 -0.9667>}
 middle_metacarpal_r {parent=hand_r, loc=<-2.1856 0.1726 0.5984>, rot=<0.006 0.0168 -0.0329 -0.9993>}
 middle_01_r {parent=middle_metacarpal_r, loc=<-7.3273 0.2447 -0.0129>, rot=<-0.0208 0.0042 0.0858 -0.9961>}
 middle_02_r {parent=middle_01_r, loc=<-4.0142 -0.1411 -0.07>, rot=<-0.0217 0.0259 0.3512 -0.9357>}
 middle_03_r {parent=middle_02_r, loc=<-2.6074 -0.0534 -0.1312>, rot=<0.0083 0.0188 0.1708 -0.9851>}
 pinky_metacarpal_r {parent=hand_r, loc=<-1.7617 0.9753 -3.3814>, rot=<-0.0397 0.0814 -0.069 -0.9935>}
 pinky_01_r {parent=pinky_metacarpal_r, loc=<-7.1707 0.4832 0.1094>, rot=<-0.0654 -0.0035 0.2242 -0.9723>}
 pinky_02_r {parent=pinky_01_r, loc=<-2.8274 -0.0892 0.1043>, rot=<0.0037 0.0011 0.2854 -0.9584>}
 pinky_03_r {parent=pinky_02_r, loc=<-2.2045 -0.031 0.0078>, rot=<-0.0002 0.0098 0.1445 -0.9895>}
 ring_metacarpal_r {parent=hand_r, loc=<-2.2073 0.3627 -1.3638>, rot=<-0.0284 0.0722 -0.0458 -0.9959>}
 ring_01_r {parent=ring_metacarpal_r, loc=<-7.0297 0.6197 0.1446>, rot=<-0.0489 -0.0252 0.1892 -0.9804>}
 ring_02_r {parent=ring_01_r, loc=<-3.6172 -0.4648 0.1027>, rot=<0 0 0.2806 -0.9598>}
 ring_03_r {parent=ring_02_r, loc=<-2.4152 -0.1141 -0.0199>, rot=<0 0 0.1189 -0.9929>}
 thumb_01_r {parent=hand_r, loc=<-1.3119 0.3192 2.459>, rot=<0.6546 -0.4042 0.0664 -0.6354>}
 thumb_02_r {parent=thumb_01_r, loc=<-5.9335 0.0447 -0.0401>, rot=<-0.0037 0.0213 0.298 -0.9543>}
 thumb_03_r {parent=thumb_02_r, loc=<-2.2223 -0.3682 0.0907>, rot=<0.0118 -0.013 0.0066 -0.9998>}
 weapon_r {parent=hand_r, loc=<0.1975 3.4111 -0.3811>, rot=<0 0 0 1>}
 lowerarm_twist_01_r {parent=lowerarm_r, loc=<-17.855 0 0>, rot=<0 0 0 1>}
 lowerarm_twist_02_r {parent=lowerarm_r, loc=<-9.6199 0 0>, rot=<0 0 0 1>}
 upperarm_twist_01_r {parent=upperarm_r, loc=<-0.0041 0 0>, rot=<0 0 0 1>}
 upperarm_twist_02_r {parent=upperarm_r, loc=<-13.7942 0 0>, rot=<0 0 0 1>}
 neck_01 {parent=spine_05, loc=<11.3138 0.1529 0>, rot=<0 0 -0.2003 -0.9797>}
 neck_02 {parent=neck_01, loc=<4.5627 0 0>, rot=<0 0 0.0504 -0.9987>}
 head {parent=neck_02, loc=<5.3914 0 0>, rot=<0 0 0.115 -0.9934>}
 thigh_l {parent=pelvis, loc=<-2.5693 -0.0109 -7.7501>, rot=<0.0012 0.0572 0.0155 -0.9982>}
 calf_l {parent=thigh_l, loc=<-39.0484 0 0>, rot=<0.0006 -0.0161 -0.0391 -0.9991>}
 calf_twist_01_l {parent=calf_l, loc=<-35.8131 -0.0967 0.0397>, rot=<0.0237 0.0095 -0.0035 -0.9997>}
 calf_twist_02_l {parent=calf_l, loc=<-18.8064 -0.0967 0.0397>, rot=<0.0237 0.0095 -0.0035 -0.9997>}
 foot_l {parent=calf_l, loc=<-42.561 0.1077 0.4292>, rot=<0.0418 -0.0455 0.0343 -0.9975>}
 ball_l {parent=foot_l, loc=<-6.1599 13.8264 0.6806>, rot=<0.0001 -0.0001 -0.7063 -0.7079>}
 thigh_twist_01_l {parent=thigh_l, loc=<-21.6303 0 0>, rot=<0.0013 0 0 -1>}
 dyn_kneePit_Base_l {parent=thigh_l, loc=<-37.092 -7.8355 -0.699>, rot=<-0.6764 -0.7346 -0.0456 0.0285>}
 dyn_kneePit_l {parent=dyn_kneePit_Base_l, loc=<0 0 0>, rot=<0 0 0 -1>}
 thigh_r {parent=pelvis, loc=<-2.5691 -0.0109 7.7501>, rot=<0.0572 -0.0012 0.9982 0.0155>}
 calf_r {parent=thigh_r, loc=<39.0487 0 0>, rot=<0.0006 -0.0161 -0.0391 -0.9991>}
 calf_twist_01_r {parent=calf_r, loc=<35.8132 0.0967 -0.0397>, rot=<0.0237 0.0095 -0.0035 -0.9997>}
 calf_twist_02_r {parent=calf_r, loc=<18.8064 0.0967 -0.0397>, rot=<0.0237 0.0095 -0.0035 -0.9997>}
 foot_r {parent=calf_r, loc=<42.5611 -0.1077 -0.4292>, rot=<0.0418 -0.0455 0.0343 -0.9975>}
 ball_r {parent=foot_r, loc=<6.1598 -13.8264 -0.6806>, rot=<0.0001 -0.0001 -0.7063 -0.7079>}
 thigh_twist_01_r {parent=thigh_r, loc=<21.6303 0 0>, rot=<0.0013 0 0 -1>}
 dyn_kneePit_Base_r {parent=thigh_r, loc=<37.0922 7.8355 0.6989>, rot=<-0.6764 -0.7346 -0.0456 0.0285>}
 dyn_kneePit_r {parent=dyn_kneePit_Base_r, loc=<0 0 0>, rot=<0 0 0 1>}
 dyn_thighScale_r {parent=pelvis, loc=<-2.5691 -0.0109 7.7501>, rot=<0.0572 -0.0012 0.9982 0.0155>}
 dyn_thighScale_l {parent=pelvis, loc=<-2.5693 -0.0109 -7.7501>, rot=<0.0012 0.0572 0.0155 -0.9982>}
 dyn_coatFront01_Base_l {parent=pelvis, loc=<14.1005 6.9126 -11.6695>, rot=<0.5765 0.081 0.0645 -0.8105>}
 dyn_coatFront01_l {parent=dyn_coatFront01_Base_l, loc=<0 0 0>, rot=<0 0 0 -1>}
 dyn_coatFront02_l {parent=dyn_coatFront01_l, loc=<-11.7766 0 0>, rot=<0.0281 -0.0113 0.02 -0.9993>}
 dyn_coatFront03_l {parent=dyn_coatFront02_l, loc=<-10.6492 0 0>, rot=<-0.0353 -0.0148 0.0178 -0.9991>}
 dyn_coatFront04_l {parent=dyn_coatFront03_l, loc=<-9.2307 0 0>, rot=<-0.0182 -0.0008 0.027 -0.9995>}
 dyn_coatFront05_l {parent=dyn_coatFront04_l, loc=<-10.3171 0 0>, rot=<0.0639 -0.0149 -0.057 -0.9962>}
 dyn_coatFront06_l {parent=dyn_coatFront05_l, loc=<-7.9977 0 0>, rot=<0 0 0 1>}
 dyn_coatSide01_Base_l {parent=pelvis, loc=<6.4002 -1.4931 -13.8882>, rot=<0.8275 0.0811 0.1142 -0.5437>}
 dyn_coatSide01_l {parent=dyn_coatSide01_Base_l, loc=<0 0 0>, rot=<0 0 0 -1>}
 dyn_coatSide02_l {parent=dyn_coatSide01_l, loc=<-14.7033 0 0>, rot=<0.0513 0.0205 -0.0017 -0.9985>}
 dyn_coatSide03_l {parent=dyn_coatSide02_l, loc=<-9.5517 0 0>, rot=<-0.0217 0.006 -0.014 -0.9996>}
 dyn_coatSide04_l {parent=dyn_coatSide03_l, loc=<-11.6513 0 0>, rot=<-0.0494 -0.0101 -0.0423 -0.9978>}
 dyn_coatSide05_l {parent=dyn_coatSide04_l, loc=<-9.6047 0 0>, rot=<0.0175 0 0 -0.9998>}
 dyn_coatBack01_Base_l {parent=pelvis, loc=<2.2417 -8.181 -11.3206>, rot=<0.9358 0.0511 0.0766 -0.3404>}
 dyn_coatBack01_l {parent=dyn_coatBack01_Base_l, loc=<0 0 0>, rot=<0 0 0 1>}
 dyn_coatBack02_l {parent=dyn_coatBack01_l, loc=<-10.8306 0 0>, rot=<0.0061 0.0045 0.0357 -0.9993>}
 dyn_coatBack03_l {parent=dyn_coatBack02_l, loc=<-10.4153 0 0>, rot=<0.0028 0.0002 0.0017 -1>}
 dyn_coatBack04_l {parent=dyn_coatBack03_l, loc=<-12.6854 0 0>, rot=<0.001 0.0025 -0.0391 -0.9992>}
 dyn_coatBack05_l {parent=dyn_coatBack04_l, loc=<-10.7094 0 0.0001>, rot=<0.0175 0 0 -0.9998>}
 dyn_coatTail01_Base_l {parent=pelvis, loc=<1.6046 -11.3679 -6.1232>, rot=<0.9848 0.1051 0.0196 -0.1371>}
 dyn_coatTail01_l {parent=dyn_coatTail01_Base_l, loc=<0 0 0>, rot=<0 0 0 1>}
 dyn_coatTail02_l {parent=dyn_coatTail01_l, loc=<-7.2778 0 0>, rot=<-0.0211 -0.0273 -0.0807 -0.9961>}
 dyn_coatTail03_l {parent=dyn_coatTail02_l, loc=<-10.8612 0 0>, rot=<-0.0016 -0.0415 0.0773 -0.9961>}
 dyn_coatTail04_l {parent=dyn_coatTail03_l, loc=<-12.6713 0 0>, rot=<0.0015 0.0019 -0.0026 -1>}
 dyn_coatTail05_l {parent=dyn_coatTail04_l, loc=<-10.3923 0 0>, rot=<-0.0283 0.007 -0.028 -0.9992>}
 dyn_coatTail06_l {parent=dyn_coatTail05_l, loc=<-6.4453 0 0>, rot=<0 0 0 1>}
 dyn_splitDown_l {parent=dyn_coatTail05_l, loc=<-0.4322 -0.1652 -2.9198>, rot=<0.0646 0.6817 -0.0579 -0.7265>}
 dyn_splitUp_l {parent=dyn_coatTail04_l, loc=<-0.2332 -0.1236 -2.8813>, rot=<0.0014 0.6957 -0.0413 -0.7172>}
 dyn_strap_03_Base_l {parent=dyn_coatTail04_l, loc=<-16.1587 1.838 1.1731>, rot=<0.0457 -0.032 0.0057 -0.9984>}
 dyn_strap_03_l {parent=dyn_strap_03_Base_l, loc=<0 0 0>, rot=<0 0 0 -1>}
 dyn_strap_04_l {parent=dyn_strap_03_l, loc=<-13.027 0 0>, rot=<0 0 0 -1>}
 dyn_strap_05_l {parent=dyn_strap_04_l, loc=<-9.906 0 0>, rot=<0.0221 -0.0022 0.0136 -0.9997>}
 dyn_strap_06_l {parent=dyn_strap_05_l, loc=<-9.7405 -0.0001 0>, rot=<0 0 0 1>}
 dyn_strap_01_Base_l {parent=dyn_coatTail02_l, loc=<-11.1339 1.5513 -0.5553>, rot=<0.0432 -0.075 0.0784 -0.9932>}
 dyn_strap_01_l {parent=dyn_strap_01_Base_l, loc=<0 0 0>, rot=<0 0 0 1>}
 dyn_strap_02_l {parent=dyn_strap_01_l, loc=<-12.2724 0 0>, rot=<0.0001 0 0 -1>}
 dyn_coatSide01_Base_r {parent=pelvis, loc=<6.4002 -1.4931 13.8882>, rot=<0.0811 -0.8275 0.5437 0.1142>}
 dyn_coatSide01_r {parent=dyn_coatSide01_Base_r, loc=<0 0 0>, rot=<0 0 0 -1>}
 dyn_coatSide02_r {parent=dyn_coatSide01_r, loc=<14.7032 0 0>, rot=<0.0513 0.0205 -0.0017 -0.9985>}
 dyn_coatSide03_r {parent=dyn_coatSide02_r, loc=<9.5517 0 0>, rot=<-0.0217 0.006 -0.014 -0.9996>}
 dyn_coatSide04_r {parent=dyn_coatSide03_r, loc=<11.6514 0 0>, rot=<-0.0494 -0.0101 -0.0423 -0.9978>}
 dyn_coatSide05_r {parent=dyn_coatSide04_r, loc=<9.6047 0 0>, rot=<0.0175 0 0 -0.9998>}
 dyn_splitUp_01_Base_r {parent=pelvis, loc=<-28.604 -15.1052 9.9169>, rot=<0.1393 -0.7791 0.067 -0.6076>}
 dyn_splitUp_01_r {parent=dyn_splitUp_01_Base_r, loc=<0 0 0>, rot=<0 0 0 -1>}
 dyn_splitDown_01_Base_r {parent=pelvis, loc=<-38.6838 -16.6387 11.928>, rot=<0.1493 -0.7902 0.0862 -0.5882>}
 dyn_splitDown_01_r {parent=dyn_splitDown_01_Base_r, loc=<0 0 0>, rot=<0 0 0 -1>}
 dyn_coatTail01_Base_r {parent=pelvis, loc=<1.6046 -11.368 6.1232>, rot=<0.1051 -0.9848 0.1371 0.0196>}
 dyn_coatTail01_r {parent=dyn_coatTail01_Base_r, loc=<0 0 0>, rot=<0 0 0 1>}
 dyn_coatTail02_r {parent=dyn_coatTail01_r, loc=<7.2778 0 0>, rot=<-0.0211 -0.0273 -0.0807 -0.9961>}
 dyn_coatTail03_r {parent=dyn_coatTail02_r, loc=<10.8612 0 0>, rot=<-0.0016 -0.0415 0.0773 -0.9961>}
 dyn_coatTail04_r {parent=dyn_coatTail03_r, loc=<12.6714 0 0>, rot=<0.0015 0.0019 -0.0026 -1>}
 dyn_coatTail05_r {parent=dyn_coatTail04_r, loc=<10.3922 0 0>, rot=<-0.0283 0.007 -0.028 -0.9992>}
 dyn_coatTail06_r {parent=dyn_coatTail05_r, loc=<6.4454 0 0>, rot=<0 0 0 1>}
 dyn_splitDown_r {parent=dyn_coatTail05_r, loc=<0.4322 0.1652 2.9198>, rot=<0.0646 0.6817 -0.0579 -0.7265>}
 dyn_splitUp_r {parent=dyn_coatTail04_r, loc=<0.2332 0.1237 2.8813>, rot=<0.0014 0.6957 -0.0413 -0.7172>}
 dyn_strap_03_Base_r {parent=dyn_coatTail04_r, loc=<16.1589 -1.838 -1.1731>, rot=<0.0457 -0.032 0.0057 -0.9984>}
 dyn_strap_03_r {parent=dyn_strap_03_Base_r, loc=<0 0 0>, rot=<0 0 0 -1>}
 dyn_strap_04_r {parent=dyn_strap_03_r, loc=<13.0266 0.0001 0.0001>, rot=<0 0 0 -1>}
 dyn_strap_05_r {parent=dyn_strap_04_r, loc=<9.9057 0.0001 0>, rot=<0.0221 -0.0022 0.0136 -0.9997>}
 dyn_strap_06_r {parent=dyn_strap_05_r, loc=<9.7405 0 0>, rot=<0 0 0 1>}
 dyn_strap_01_Base_r {parent=dyn_coatTail02_r, loc=<11.1339 -1.5513 0.5553>, rot=<0.0432 -0.075 0.0784 -0.9932>}
 dyn_strap_01_r {parent=dyn_strap_01_Base_r, loc=<0 0 0>, rot=<0 0 0 1>}
 dyn_strap_02_r {parent=dyn_strap_01_r, loc=<12.2724 0 0>, rot=<0.0001 0 0 -1>}
 dyn_coatSplit01_Base {parent=pelvis, loc=<1.1612 -12.3328 0>, rot=<0 0 0.6408 -0.7677>}
 dyn_coatSplit01 {parent=dyn_coatSplit01_Base, loc=<0 0 0>, rot=<0 0 0 -1>}
 dyn_coatSplit02 {parent=dyn_coatSplit01, loc=<0.6852 0 0>, rot=<0.7097 -0.7045 0 0>}
 dyn_coatSplit03 {parent=dyn_coatSplit02, loc=<10.7424 0 0>, rot=<0 0 -0.0765 -0.9971>}
 dyn_coatBack01_Base_r {parent=pelvis, loc=<2.2417 -8.181 11.3206>, rot=<0.0511 -0.9358 0.3404 0.0766>}
 dyn_coatBack01_r {parent=dyn_coatBack01_Base_r, loc=<0 0 0>, rot=<0 0 0 -1>}
 dyn_coatBack02_r {parent=dyn_coatBack01_r, loc=<10.8306 0 0>, rot=<0.0061 0.0045 0.0357 -0.9993>}
 dyn_coatBack03_r {parent=dyn_coatBack02_r, loc=<10.4153 0 0>, rot=<0.0028 0.0002 0.0017 -1>}
 dyn_coatBack04_r {parent=dyn_coatBack03_r, loc=<12.6855 0 0>, rot=<0.001 0.0025 -0.0391 -0.9992>}
 dyn_coatBack05_r {parent=dyn_coatBack04_r, loc=<10.7093 0 0>, rot=<0.0175 0 0 -0.9998>}
 dyn_coatFront01_Base_r {parent=pelvis, loc=<14.1003 6.9126 11.6695>, rot=<0.081 -0.5765 0.8105 0.0645>}
 dyn_coatFront01_r {parent=dyn_coatFront01_Base_r, loc=<0 0 0>, rot=<0 0 0 1>}
 dyn_coatFront02_r {parent=dyn_coatFront01_r, loc=<11.7763 0 0>, rot=<0.0281 -0.0113 0.02 -0.9993>}
 dyn_coatFront03_r {parent=dyn_coatFront02_r, loc=<10.6492 0 0>, rot=<-0.0353 -0.0148 0.0178 -0.9991>}
 dyn_coatFront04_r {parent=dyn_coatFront03_r, loc=<9.2306 0 0>, rot=<-0.0182 -0.0008 0.027 -0.9995>}
 dyn_coatFront05_r {parent=dyn_coatFront04_r, loc=<10.3172 0 0>, rot=<0.0639 -0.0149 -0.057 -0.9962>}
 dyn_coatFront06_r {parent=dyn_coatFront05_r, loc=<7.9977 0 0>, rot=<0 0 0 -1>}
 ik_foot_root {parent=root, loc=<0 0 0>, rot=<0 0 0 1>}
 ik_foot_l {parent=ik_foot_root, loc=<15.2608 -4.4176 9.4433>, rot=<-0.0357 -0.7045 -0.0284 0.7082>}
 ik_foot_r {parent=ik_foot_root, loc=<-15.2609 -4.4176 9.4431>, rot=<-0.7082 -0.0284 0.7045 -0.0357>}
 ik_hand_root {parent=root, loc=<0 0 0>, rot=<0 0 0 1>}
 ik_hand_gun {parent=ik_hand_root, loc=<-51.973 4.089 102.694>, rot=<-0.6084 -0.1793 -0.4613 0.6205>}
 ik_hand_l {parent=ik_hand_gun, loc=<53.0284 82.1785 35.2051>, rot=<-0.5102 -0.7906 -0.3387 0>}
 ik_hand_r {parent=ik_hand_gun, loc=<0 0 0>, rot=<0 0 0 -1>}
]
