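# Physics asset for the Shiitake Shaolin male body's falling variant, setting up secondary motion for coat panels, tails and straps.
# Pelvis-anchored front, side, back, split and tail chains use limited or locked joints, with straps and split flaps branching off the tails.
# Free cross-constraints tie neighboring panels together.
# Leg, pelvis and spine shapes provide body volumes, root adds a huge offset sphere, and lower left-tail pieces use capsules where the right uses boxes.
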PhysicsAsset /BRCosmetics/Characters/Player/Male/Medium/Bodies/M_MED_Shiitake_Shaolin/Meshes/M_MED_Shiitake_Shaolin_Falling_Physics
bodies {
 calf_l [{shapeType=Capsule, length=12.8527, radius=4.2889, rotation=<rot 90 0 0>}]
 calf_twist_01_l [{shapeType=Capsule, length=23.7657, radius=4.9996, center=<1.6026 -1.5489 -0.7212>, rotation=<rot 90 0 -171.9356>}]
 calf_twist_01_r [{shapeType=Capsule, length=30.2146, radius=6.3938, center=<-4.2942 2.5803 -0.0001>, rotation=<rot 90 15.3763 23.4407>}]
 calf_twist_02_l [{shapeType=Capsule, length=13.6414, radius=5.7083, center=<4.0549 -5.2499 -1.4295>, rotation=<rot 90 0 0>}]
 calf_twist_02_r [{shapeType=Capsule, length=18.2358, radius=6.5846, center=<-4.5916 5.2499 1.429>, rotation=<rot 90 0 -179.9999>}]
 dyn_coatBack01_Base_l [{shapeType=Box, extentX=1.01, extentY=1.01, extentZ=1.01, rotation=<rot 0 -90 180>}]
 dyn_coatBack01_Base_r [{shapeType=Box, extentX=1.01, extentY=1.01, extentZ=1.01, rotation=<rot 0 -90 0>}]
 dyn_coatBack01_l [{shapeType=Box, extentX=1.9529, extentY=12.4841, extentZ=11.5709, center=<-5.5629 -0.1478 -0.3402>, rotation=<rot 0 90 180>}]
 dyn_coatBack01_r [{shapeType=Box, extentX=1.9529, extentY=12.4841, extentZ=11.5709, center=<5.5629 0.1478 0.3402>, rotation=<rot 0 90 0>}]
 dyn_coatBack02_l [{shapeType=Box, extentX=1.9529, extentY=12.1879, extentZ=11.5709, center=<-4.8319 0 -0.3402>, rotation=<rot 0 89.9999 -180>}]
 dyn_coatBack02_r [{shapeType=Box, extentX=2.5583, extentY=12.1879, extentZ=11.5709, center=<4.8319 0 0>, rotation=<rot 0 90 0>}]
 dyn_coatBack03_l [{shapeType=Box, extentX=1.9529, extentY=14.1977, extentZ=11.5709, center=<-6.3607 -0.0001 -0.3402>, rotation=<rot 0 90 180>}]
 dyn_coatBack03_r [{shapeType=Box, extentX=1.9529, extentY=14.1977, extentZ=11.5709, center=<6.3607 0 0>, rotation=<rot 0 90 0>}]
 dyn_coatBack04_l [{shapeType=Box, extentX=1.9529, extentY=12.703, extentZ=13.4915, center=<-5.343 0 0>, rotation=<rot 0 90 180>}]
 dyn_coatBack04_r [{shapeType=Box, extentX=3.2093, extentY=13.3615, extentZ=13.0098, center=<6.1756 0 0>, rotation=<rot 0 90 0>}]
 dyn_coatFront01_Base_l [{shapeType=Box, extentX=1.01, extentY=1.01, extentZ=1.01, rotation=<rot 0 -90 180>}]
 dyn_coatFront01_Base_r [{shapeType=Box, extentX=1.01, extentY=1.01, extentZ=1.01, rotation=<rot 0 -90 0>}]
 dyn_coatFront01_l [{shapeType=Box, extentX=1.9529, extentY=12.8046, extentZ=13.8966, center=<-5.8542 -0.0001 0>, rotation=<rot 0 90 180>}]
 dyn_coatFront01_r [{shapeType=Box, extentX=1.9529, extentY=12.8046, extentZ=13.7058, center=<5.8542 0.0001 0>, rotation=<rot 0 90 0>}]
 dyn_coatFront02_l [{shapeType=Box, extentX=1.9465, extentY=10.2915, extentZ=13.3636, center=<-4.7421 -0.3976 -0.0286>, rotation=<rot 4.122 89.9999 180>}]
 dyn_coatFront02_r [{shapeType=Box, extentX=1.9465, extentY=10.2915, extentZ=13.3636, center=<4.7421 0.3976 0.0286>, rotation=<rot 4.122 89.9999 0>}]
 dyn_coatFront03_l [{shapeType=Box, extentX=2.203, extentY=10.5565, extentZ=14.2678, center=<-4.1681 -0.1865 0>, rotation=<rot 1.4145 89.9999 -180>}]
 dyn_coatFront03_r [{shapeType=Box, extentX=2.203, extentY=10.5565, extentZ=14.2678, center=<4.1681 0.1865 0>, rotation=<rot 1.4145 89.9999 0>}]
 dyn_coatFront04_l [{shapeType=Box, extentX=2.4066, extentY=10.5327, extentZ=14.3522, center=<-4.7296 -0.1354 0>, rotation=<rot -2.0393 91.5601 -180>}]
 dyn_coatFront04_r [{shapeType=Box, extentX=2.4066, extentY=10.5327, extentZ=14.3522, center=<4.7296 0.1354 0>, rotation=<rot -2.0393 91.5601 0>}]
 dyn_coatFront05_l [{shapeType=Box, extentX=2.6034, extentY=9.7361, extentZ=14.3831, center=<-4.1029 -0.2971 0>, rotation=<rot 3.8167 89.9999 180>}]
 dyn_coatFront05_r [{shapeType=Box, extentX=3.1806, extentY=9.7361, extentZ=14.3831, center=<4.1029 0.2971 0>, rotation=<rot 3.8167 89.9999 0>}]
 dyn_coatSide01_Base_l [{shapeType=Box, extentX=1.01, extentY=1.01, extentZ=1.01, rotation=<rot 0 -90 180>}]
 dyn_coatSide01_Base_r [{shapeType=Box, extentX=1.01, extentY=1.01, extentZ=1.01, rotation=<rot 0 -90 0>}]
 dyn_coatSide01_l [{shapeType=Box, extentX=2.1372, extentY=16.1439, extentZ=11.5709, center=<-7.5606 1.0187 -0.3402>, rotation=<rot 0 91.9573 -180>}]
 dyn_coatSide01_r [{shapeType=Box, extentX=2.1372, extentY=16.1439, extentZ=11.5709, center=<7.5606 -1.0187 0.3402>, rotation=<rot 0 90.5035 0>}]
 dyn_coatSide02_l [{shapeType=Box, extentX=1.9551, extentY=10.3497, extentZ=12.948, center=<-4.8677 0.0796 -0.5534>, rotation=<rot 8.1678 89.9995 175.3491>}]
 dyn_coatSide02_r [{shapeType=Box, extentX=1.9551, extentY=10.3497, extentZ=12.948, center=<4.8677 -0.0796 0.5534>, rotation=<rot 8.1678 89.9995 -4.6509>}]
 dyn_coatSide03_l [{shapeType=Box, extentX=2.3571, extentY=12.5399, extentZ=13.151, center=<-5.6311 0.0001 0>, rotation=<rot 8.2451 90 173.9941>}]
 dyn_coatSide03_r [{shapeType=Box, extentX=2.5374, extentY=12.5399, extentZ=13.151, center=<5.6311 -0.0001 0>, rotation=<rot 8.2451 90 -6.0059>}]
 dyn_coatSide04_l [{shapeType=Box, extentX=2.1868, extentY=9.9901, extentZ=12.0801, center=<-4.6982 -0.0201 0.1442>, rotation=<rot 7.8241 89.9998 171.0141>}]
 dyn_coatSide04_r [{shapeType=Box, extentX=3.1925, extentY=9.9901, extentZ=12.8853, center=<4.6982 0.0201 -0.1442>, rotation=<rot 7.8241 89.9998 -8.9859>}]
 dyn_coatSplit01 [{shapeType=Box, extentX=1.9529, extentY=2.7505, extentZ=1.8502, center=<0.4581 -0.0002 0>, rotation=<rot 0 90 0>}]
 dyn_coatSplit01_Base [{shapeType=Box, extentX=1.01, extentY=1.01, extentZ=1.01, rotation=<rot 0 -90 0>}]
 dyn_coatSplit02 [{shapeType=Box, extentX=1.9529, extentY=12.1392, extentZ=14.6415, center=<5.6462 -0.0002 0>, rotation=<rot 0 90 0>}]
 dyn_coatSplit03 [{shapeType=Box, extentX=1.9529, extentY=13.9974, extentZ=14.6415, center=<6.4038 -0.0003 0>, rotation=<rot 0 90 0>}]
 dyn_coatTail01_Base_l [{shapeType=Box, extentX=1.9529, extentY=2.7802, extentZ=3.2927, center=<-0.014 0 -0.3402>, rotation=<rot 0 90 180>}]
 dyn_coatTail01_Base_r [{shapeType=Box, extentX=1.9529, extentY=2.7802, extentZ=3.2927, center=<0.014 0 0.3402>, rotation=<rot 0 90 0>}]
 dyn_coatTail01_l [{shapeType=Box, extentX=1.9529, extentY=9.234, extentZ=11.5709, center=<-3.8416 -0.0246 0.0847>, rotation=<rot 3.2986 89.9999 180>}]
 dyn_coatTail01_r [{shapeType=Box, extentX=1.9529, extentY=9.234, extentZ=11.5709, center=<3.8416 0.0246 -0.0847>, rotation=<rot 3.2986 89.9999 0>}]
 dyn_coatTail02_l [{shapeType=Box, extentX=1.9529, extentY=12.8979, extentZ=14.6415, center=<-5.3435 0.0002 0>, rotation=<rot 0 89.9999 -180>}]
 dyn_coatTail02_r [{shapeType=Box, extentX=2.8259, extentY=12.8979, extentZ=14.6415, center=<5.3435 -0.0002 0>, rotation=<rot 0 90 0>}]
 dyn_coatTail03_l [{shapeType=Capsule, length=8.25, radius=3.75}]
 dyn_coatTail03_r [{shapeType=Box, extentX=1.9529, extentY=14.4964, extentZ=15.7613, center=<6.2222 0 0.3402>, rotation=<rot 0 90 0>}]
 dyn_coatTail04_l [{shapeType=Capsule, length=8.25, radius=4}]
 dyn_coatTail04_r [{shapeType=Box, extentX=1.9529, extentY=11.6183, extentZ=17.2674, center=<5.8251 0.0001 0.3402>, rotation=<rot 0 90 0>}]
 dyn_coatTail05_l [{shapeType=Capsule, length=7.75, radius=3.5, center=<-1 0 0>}]
 dyn_coatTail05_r [{shapeType=Box, extentX=2.4369, extentY=10.3226, extentZ=17.7112, center=<4.5002 0.0001 0.1395>, rotation=<rot 0 90 0>}]
 dyn_splitDown_l [{shapeType=Capsule, length=8, radius=2.5}]
 dyn_splitDown_r [{shapeType=Box, extentX=3.107, extentY=9.0597, extentZ=13.7756, center=<3 0 -1>, rotation=<rot 0 89.9999 2.9661>}]
 dyn_splitUp_l [{shapeType=Capsule, length=7.75, radius=3}]
 dyn_splitUp_r [{shapeType=Box, extentX=1.9529, extentY=7.5351, extentZ=11.5709, center=<2.4309 0 0>, rotation=<rot 0 90 0>}]
 dyn_strap_03_l [{shapeType=Capsule, length=8.25, radius=2.5, center=<-4.0686 0 0>, rotation=<rot 90 -0.0297 -0.0299>}]
 dyn_strap_03_r [{shapeType=Box, extentX=1.3343, extentY=5.9251, extentZ=13.8514, center=<6.0761 0 0>, rotation=<rot -7.3705 -89.9999 -89.9994>}]
 dyn_strap_04_l [{shapeType=Capsule, length=4.5, radius=2.75, center=<-3.7209 0 0>, rotation=<rot 90 0 0>}]
 dyn_strap_04_r [{shapeType=Box, extentX=1.5362, extentY=10.5589, extentZ=5.9074, center=<4.8954 0 0>, rotation=<rot -7.76 -90 0.0155>}]
 dyn_strap_05_l [{shapeType=Capsule, length=4.5, radius=2.75, center=<-3.7209 0 0>, rotation=<rot 90 0 0>}]
 dyn_strap_05_r [{shapeType=Box, extentX=1.6105, extentY=5.9119, extentZ=10.6149, center=<4.8666 0 0>, rotation=<rot 12.5061 91.7523 -89.6863>}]
 foot_l [{shapeType=Capsule, length=4.2601, radius=4.2889, center=<-3.5689 -1.6323 0>, rotation=<rot 90 0 180>}]
 foot_r [{shapeType=Capsule, length=4.2601, radius=4.2889, center=<3.5689 1.6323 0>, rotation=<rot 90 0 0>}]
 pelvis [{shapeType=Capsule, length=11.1226, radius=13.4575, center=<-0.0001 1.3847 0>, rotation=<rot 90 0 0>}]
 root [
  {shapeType=Sphere, radius=190, shapeName=root_sphere, center=<-3.4923 -2.7141 -175>}
  {shapeType=Sphere, radius=24.9585, shapeName=root_sphere, center=<0 2.4787 8.5972>}
 ]
 spine_01 [{shapeType=Capsule, length=15.6372, radius=8.2031, rotation=<rot 90 0 0>}]
 thigh_l [{shapeType=Capsule, length=10.6901, radius=8.387, rotation=<rot 81.1219 -180 -180>}]
 thigh_r [{shapeType=Capsule, length=10.6901, radius=8.387, rotation=<rot 84.0249 180 180>}]
 thigh_twist_01_l [
  {shapeType=Sphere, radius=10.1625, center=<-13.5254 -0.914 0.0001>}
  {shapeType=Capsule, length=26.3738, radius=10.0644, center=<1.3326 0 0>, rotation=<rot 90 0 -179.9999>}
 ]
 thigh_twist_01_r [
  {shapeType=Sphere, radius=12.2189, center=<13.5254 2.588 0.1043>}
  {shapeType=Capsule, length=26.3738, radius=10.0644, center=<-1.7162 0 0>, rotation=<rot 90 0 0>}
 ]
}
constraints [
 {bone1Name=dyn_coatFront01_r, bone2Name=dyn_coatFront01_Base_r, swing1LimitDegrees=30, swing2LimitDegrees=20, twistLimitDegrees=8}
 {bone1Name=dyn_coatFront02_r, bone2Name=dyn_coatFront01_r, swing1LimitDegrees=30, swing2LimitDegrees=20, twistLimitDegrees=12}
 {bone1Name=dyn_coatFront03_r, bone2Name=dyn_coatFront02_r, swing1LimitDegrees=60, swing2LimitDegrees=20, twistLimitDegrees=12}
 {bone1Name=dyn_coatFront04_r, bone2Name=dyn_coatFront03_r, swing1LimitDegrees=60, swing2LimitDegrees=20, twistLimitDegrees=12}
 {bone1Name=dyn_coatFront05_r, bone2Name=dyn_coatFront04_r, swing1LimitDegrees=60, swing2LimitDegrees=20, twistLimitDegrees=12}
 {bone1Name=dyn_coatFront01_l, bone2Name=dyn_coatFront01_Base_l, swing1LimitDegrees=24, swing2LimitDegrees=16, twistLimitDegrees=8}
 {bone1Name=dyn_coatFront02_l, bone2Name=dyn_coatFront01_l, swing1LimitDegrees=24, swing2LimitDegrees=16, twistLimitDegrees=12}
 {bone1Name=dyn_coatFront03_l, bone2Name=dyn_coatFront02_l, swing1LimitDegrees=60, swing2LimitDegrees=16, twistLimitDegrees=12}
 {bone1Name=dyn_coatFront04_l, bone2Name=dyn_coatFront03_l, swing1LimitDegrees=60, swing2LimitDegrees=20, twistLimitDegrees=12}
 {bone1Name=dyn_coatFront05_l, bone2Name=dyn_coatFront04_l, swing1LimitDegrees=60, swing2LimitDegrees=20, twistLimitDegrees=12}
 {
  bone1Name=thigh_twist_01_r
  bone2Name=thigh_r
  swing1LimitDegrees=45
  swing1Motion=Free
  swing2LimitDegrees=45
  swing2Motion=Free
  twistLimitDegrees=45
  twistMotion=Free
 }
 {
  bone1Name=thigh_l
  bone2Name=pelvis
  swing1LimitDegrees=45
  swing1Motion=Free
  swing2LimitDegrees=45
  swing2Motion=Free
  twistLimitDegrees=45
  twistMotion=Free
 }
 {
  bone1Name=thigh_r
  bone2Name=pelvis
  swing1LimitDegrees=45
  swing1Motion=Free
  swing2LimitDegrees=45
  swing2Motion=Free
  twistLimitDegrees=45
  twistMotion=Free
 }
 {
  bone1Name=dyn_coatFront01_Base_l
  bone2Name=pelvis
  swing1LimitDegrees=45
  swing1Motion=Free
  swing2LimitDegrees=45
  swing2Motion=Free
  twistLimitDegrees=45
  twistMotion=Free
 }
 {
  bone1Name=dyn_coatFront01_Base_r
  bone2Name=pelvis
  swing1LimitDegrees=45
  swing1Motion=Free
  swing2LimitDegrees=45
  swing2Motion=Free
  twistLimitDegrees=45
  twistMotion=Free
 }
 {
  bone1Name=spine_01
  bone2Name=pelvis
  swing1LimitDegrees=45
  swing1Motion=Free
  swing2LimitDegrees=45
  swing2Motion=Free
  twistLimitDegrees=45
  twistMotion=Free
 }
 {
  bone1Name=dyn_coatSide01_Base_r
  bone2Name=pelvis
  swing1LimitDegrees=45
  swing1Motion=Free
  swing2LimitDegrees=45
  swing2Motion=Free
  twistLimitDegrees=45
  twistMotion=Free
 }
 {bone1Name=dyn_coatSide01_r, bone2Name=dyn_coatSide01_Base_r, swing1LimitDegrees=30, swing2LimitDegrees=20, twistLimitDegrees=8}
 {bone1Name=dyn_coatSide02_r, bone2Name=dyn_coatSide01_r, swing1LimitDegrees=60, swing2LimitDegrees=20, twistLimitDegrees=8}
 {bone1Name=dyn_coatSide03_r, bone2Name=dyn_coatSide02_r, swing1LimitDegrees=60, swing2LimitDegrees=20, twistLimitDegrees=8}
 {bone1Name=dyn_coatSide04_r, bone2Name=dyn_coatSide03_r, swing1LimitDegrees=60, swing2LimitDegrees=20, twistLimitDegrees=8}
 {
  bone1Name=dyn_coatSide03_r
  bone2Name=dyn_coatFront04_r
  swing1LimitDegrees=10
  swing1Motion=Free
  swing2LimitDegrees=10
  swing2Motion=Free
  twistLimitDegrees=10
  twistMotion=Free
 }
 {
  bone1Name=dyn_coatSide01_Base_l
  bone2Name=pelvis
  swing1LimitDegrees=45
  swing1Motion=Free
  swing2LimitDegrees=45
  swing2Motion=Free
  twistLimitDegrees=45
  twistMotion=Free
 }
 {bone1Name=dyn_coatSide02_l, bone2Name=dyn_coatSide01_l, swing1LimitDegrees=60, swing2LimitDegrees=16, twistLimitDegrees=8}
 {bone1Name=dyn_coatSide03_l, bone2Name=dyn_coatSide02_l, swing1LimitDegrees=60, swing2LimitDegrees=20, twistLimitDegrees=8}
 {bone1Name=dyn_coatSide04_l, bone2Name=dyn_coatSide03_l, swing1LimitDegrees=60, swing2LimitDegrees=20, twistLimitDegrees=8}
 {
  bone1Name=dyn_coatSide03_l
  bone2Name=dyn_coatFront04_l
  swing1LimitDegrees=45
  swing1Motion=Free
  swing2LimitDegrees=45
  swing2Motion=Free
  twistLimitDegrees=45
  twistMotion=Free
 }
 {
  bone1Name=dyn_coatBack01_Base_r
  bone2Name=pelvis
  swing1LimitDegrees=45
  swing1Motion=Free
  swing2LimitDegrees=45
  swing2Motion=Free
  twistLimitDegrees=45
  twistMotion=Free
 }
 {bone1Name=dyn_coatBack01_r, bone2Name=dyn_coatBack01_Base_r, swing1LimitDegrees=55, swing2LimitDegrees=20, twistLimitDegrees=8}
 {bone1Name=dyn_coatBack02_r, bone2Name=dyn_coatBack01_r, swing1LimitDegrees=55, swing2LimitDegrees=20, twistLimitDegrees=8}
 {bone1Name=dyn_coatBack03_r, bone2Name=dyn_coatBack02_r, swing1LimitDegrees=30, swing2LimitDegrees=20, twistLimitDegrees=8}
 {bone1Name=dyn_coatBack04_r, bone2Name=dyn_coatBack03_r, swing1LimitDegrees=30, swing2LimitDegrees=20, twistLimitDegrees=8}
 {
  bone1Name=dyn_coatSide03_r
  bone2Name=dyn_coatBack03_r
  swing1LimitDegrees=10
  swing1Motion=Free
  swing2LimitDegrees=10
  swing2Motion=Free
  twistLimitDegrees=10
  twistMotion=Free
 }
 {
  bone1Name=dyn_coatBack01_Base_l
  bone2Name=pelvis
  swing1LimitDegrees=45
  swing1Motion=Free
  swing2LimitDegrees=45
  swing2Motion=Free
  twistLimitDegrees=45
  twistMotion=Free
 }
 {bone1Name=dyn_coatBack01_l, bone2Name=dyn_coatBack01_Base_l, swing1LimitDegrees=55, swing2LimitDegrees=16, twistLimitDegrees=8}
 {bone1Name=dyn_coatBack02_l, bone2Name=dyn_coatBack01_l, swing1LimitDegrees=55, swing2LimitDegrees=16, twistLimitDegrees=8}
 {bone1Name=dyn_coatBack03_l, bone2Name=dyn_coatBack02_l, swing1LimitDegrees=30, swing2LimitDegrees=20, twistLimitDegrees=8}
 {bone1Name=dyn_coatBack04_l, bone2Name=dyn_coatBack03_l, swing1LimitDegrees=30, swing2LimitDegrees=20, twistLimitDegrees=8}
 {
  bone1Name=dyn_coatSide03_l
  bone2Name=dyn_coatBack03_l
  swing1LimitDegrees=45
  swing1Motion=Free
  swing2LimitDegrees=45
  swing2Motion=Free
  twistLimitDegrees=45
  twistMotion=Free
 }
 {
  bone1Name=dyn_coatSide04_r
  bone2Name=dyn_coatFront05_r
  swing1LimitDegrees=10
  swing1Motion=Free
  swing2LimitDegrees=10
  swing2Motion=Free
  twistLimitDegrees=10
  twistMotion=Free
 }
 {
  bone1Name=dyn_coatTail01_Base_r
  bone2Name=pelvis
  swing1LimitDegrees=45
  swing1Motion=Free
  swing2LimitDegrees=45
  swing2Motion=Free
  twistLimitDegrees=45
  twistMotion=Free
 }
 {bone1Name=dyn_coatTail01_r, bone2Name=dyn_coatTail01_Base_r, swing1LimitDegrees=24, swing2LimitDegrees=16, twistLimitDegrees=8}
 {
  bone1Name=dyn_coatTail02_r
  bone2Name=dyn_coatTail01_r
  swing1LimitDegrees=16
  swing2LimitDegrees=16
  twistLimitDegrees=8
  twistMotion=Locked
 }
 {
  bone1Name=dyn_coatTail03_r
  bone2Name=dyn_coatTail02_r
  swing1LimitDegrees=20
  swing2LimitDegrees=12
  swing2Motion=Locked
  twistLimitDegrees=8
  twistMotion=Locked
 }
 {bone1Name=dyn_coatTail04_r, bone2Name=dyn_coatTail03_r, swing1LimitDegrees=20, swing2LimitDegrees=12, twistLimitDegrees=12}
 {
  bone1Name=dyn_splitDown_r
  bone2Name=dyn_coatTail05_r
  swing1LimitDegrees=26
  swing2LimitDegrees=45
  swing2Motion=Locked
  twistLimitDegrees=45
  twistMotion=Locked
 }
 {bone1Name=dyn_splitUp_r, bone2Name=dyn_coatTail04_r, swing1LimitDegrees=26, swing2LimitDegrees=16, twistLimitDegrees=45, twistMotion=Locked}
 {bone1Name=dyn_coatTail05_r, bone2Name=dyn_coatTail04_r, swing1LimitDegrees=20, swing2LimitDegrees=12, twistLimitDegrees=12}
 {
  bone1Name=dyn_coatBack04_r
  bone2Name=dyn_coatTail04_r
  swing1LimitDegrees=45
  swing1Motion=Free
  swing2LimitDegrees=45
  swing2Motion=Free
  twistLimitDegrees=45
  twistMotion=Free
 }
 {
  bone1Name=dyn_coatSplit01_Base
  bone2Name=pelvis
  swing1LimitDegrees=45
  swing1Motion=Free
  swing2LimitDegrees=45
  swing2Motion=Free
  twistLimitDegrees=45
  twistMotion=Free
 }
 {
  bone1Name=dyn_coatSplit01
  bone2Name=dyn_coatSplit01_Base
  swing1LimitDegrees=45
  swing1Motion=Free
  swing2LimitDegrees=45
  swing2Motion=Free
  twistLimitDegrees=45
  twistMotion=Free
 }
 {bone1Name=dyn_coatSplit02, bone2Name=dyn_coatSplit01, swing1LimitDegrees=24, swing2LimitDegrees=16, twistLimitDegrees=8}
 {bone1Name=dyn_coatSplit03, bone2Name=dyn_coatSplit02, swing1LimitDegrees=24, swing2LimitDegrees=16, twistLimitDegrees=8}
 {
  bone1Name=dyn_coatTail03_r
  bone2Name=dyn_coatSplit03
  swing1LimitDegrees=60
  swing1Motion=Free
  swing2LimitDegrees=60
  swing2Motion=Free
  twistLimitDegrees=30
  twistMotion=Free
 }
 {bone1Name=dyn_coatSide01_l, bone2Name=dyn_coatSide01_Base_l, swing1LimitDegrees=24, swing2LimitDegrees=16, twistLimitDegrees=8}
 {
  bone1Name=dyn_coatTail01_Base_l
  bone2Name=pelvis
  swing1LimitDegrees=45
  swing1Motion=Free
  swing2LimitDegrees=45
  swing2Motion=Free
  twistLimitDegrees=45
  twistMotion=Free
 }
 {bone1Name=dyn_coatTail01_l, bone2Name=dyn_coatTail01_Base_l, swing1LimitDegrees=24, swing2LimitDegrees=16, twistLimitDegrees=8}
 {
  bone1Name=dyn_coatTail02_l
  bone2Name=dyn_coatTail01_l
  swing1LimitDegrees=16
  swing2LimitDegrees=16
  twistLimitDegrees=8
  twistMotion=Locked
 }
 {
  bone1Name=dyn_coatTail03_l
  bone2Name=dyn_coatTail02_l
  swing1LimitDegrees=20
  swing2LimitDegrees=12
  swing2Motion=Locked
  twistLimitDegrees=8
  twistMotion=Locked
 }
 {
  bone1Name=dyn_coatSplit03
  bone2Name=dyn_coatTail03_l
  swing1LimitDegrees=45
  swing1Motion=Free
  swing2LimitDegrees=45
  swing2Motion=Free
  twistLimitDegrees=45
  twistMotion=Free
 }
 {
  bone1Name=calf_l
  bone2Name=thigh_l
  swing1LimitDegrees=45
  swing1Motion=Free
  swing2LimitDegrees=45
  swing2Motion=Free
  twistLimitDegrees=45
  twistMotion=Free
 }
 {
  bone1Name=calf_twist_02_l
  bone2Name=calf_l
  swing1LimitDegrees=45
  swing1Motion=Free
  swing2LimitDegrees=45
  swing2Motion=Free
  twistLimitDegrees=45
  twistMotion=Free
 }
 {
  bone1Name=calf_twist_01_r
  bone2Name=thigh_r
  swing1LimitDegrees=45
  swing1Motion=Free
  swing2LimitDegrees=45
  swing2Motion=Free
  twistLimitDegrees=45
  twistMotion=Free
 }
 {
  bone1Name=foot_r
  bone2Name=thigh_r
  swing1LimitDegrees=45
  swing1Motion=Free
  swing2LimitDegrees=45
  swing2Motion=Free
  twistLimitDegrees=45
  twistMotion=Free
 }
 {
  bone1Name=calf_twist_02_r
  bone2Name=thigh_r
  swing1LimitDegrees=45
  swing1Motion=Free
  swing2LimitDegrees=45
  swing2Motion=Free
  twistLimitDegrees=45
  twistMotion=Free
 }
 {
  bone1Name=calf_twist_01_l
  bone2Name=calf_l
  swing1LimitDegrees=45
  swing1Motion=Free
  swing2LimitDegrees=45
  swing2Motion=Free
  twistLimitDegrees=45
  twistMotion=Free
 }
 {
  bone1Name=foot_l
  bone2Name=calf_l
  swing1LimitDegrees=45
  swing1Motion=Free
  swing2LimitDegrees=45
  swing2Motion=Free
  twistLimitDegrees=45
  twistMotion=Free
 }
 {bone1Name=dyn_coatTail04_l, bone2Name=dyn_coatTail03_l, swing1LimitDegrees=20, swing2LimitDegrees=12, twistLimitDegrees=12}
 {
  bone1Name=dyn_coatTail04_l
  bone2Name=dyn_coatBack04_l
  swing1LimitDegrees=45
  swing1Motion=Free
  swing2LimitDegrees=45
  swing2Motion=Free
  twistLimitDegrees=45
  twistMotion=Free
 }
 {bone1Name=dyn_splitDown_l, bone2Name=dyn_coatTail05_l, swing1LimitDegrees=26, swing2LimitDegrees=25, twistLimitDegrees=5}
 {bone1Name=dyn_splitUp_l, bone2Name=dyn_coatTail04_l, swing1LimitDegrees=26, swing2LimitDegrees=16, twistLimitDegrees=3}
 {bone1Name=dyn_coatTail05_l, bone2Name=dyn_coatTail04_l, swing1LimitDegrees=20, swing2LimitDegrees=12, twistLimitDegrees=12}
 {
  bone1Name=thigh_twist_01_l
  bone2Name=thigh_l
  swing1LimitDegrees=45
  swing1Motion=Free
  swing2LimitDegrees=45
  swing2Motion=Free
  twistLimitDegrees=45
  twistMotion=Free
 }
 {bone1Name=dyn_strap_03_r, bone2Name=dyn_coatTail04_r, swing1LimitDegrees=35, swing2LimitDegrees=18, twistLimitDegrees=45, twistMotion=Locked}
 {bone1Name=dyn_strap_04_r, bone2Name=dyn_strap_03_r, swing1LimitDegrees=25, swing2LimitDegrees=25, twistLimitDegrees=45, twistMotion=Locked}
 {bone1Name=dyn_strap_05_r, bone2Name=dyn_strap_04_r, swing1LimitDegrees=25, swing2LimitDegrees=25, twistLimitDegrees=16}
 {bone1Name=dyn_strap_03_l, bone2Name=dyn_coatTail04_l, swing1LimitDegrees=35, swing2LimitDegrees=18, twistLimitDegrees=5}
 {bone1Name=dyn_strap_04_l, bone2Name=dyn_strap_03_l, swing1LimitDegrees=25, swing2LimitDegrees=25, twistLimitDegrees=5}
 {bone1Name=dyn_strap_05_l, bone2Name=dyn_strap_04_l, swing1LimitDegrees=25, swing2LimitDegrees=25, twistLimitDegrees=16}
]
