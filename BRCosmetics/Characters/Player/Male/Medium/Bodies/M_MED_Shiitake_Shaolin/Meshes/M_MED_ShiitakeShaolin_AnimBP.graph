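# Post-process style AnimBlueprint that drives coat, tail and strap secondary motion for the Shiitake Shaolin player cosmetic.
# From cached 'prerigid', bRigidBodyEnabled picks pelvis-space RigidBody simulation, switched for falling and DBNO carry, or a procedural fallback.
# The fallback layers thigh-driven calf Look Ats and additive bone offsets for movement, jumping, skydiving, parachuting and crouching.
# Coat-correction offsets follow that blend; strap constraints weight neighbouring coat bones.
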AnimBlueprint /BRCosmetics/Characters/Player/Male/Medium/Bodies/M_MED_Shiitake_Shaolin/Meshes/M_MED_ShiitakeShaolin_AnimBP
target_skeleton=&/M_MED_ShiitakeShaolin_Body_Skeleton
graphs [
 AnimGraph {nodes [
   "Constraint - dyn_strap_02_r (PPPPPPPPPP)"!(
    ComponentPose
    Alpha=0.5
    ConstraintWeights_0=0.0114
    ConstraintWeights_1:float=0.2645
    ConstraintWeights_2:float=0.3158
    ConstraintWeights_3:float=0.1444
    ConstraintWeights_4:float=0.1281
    ConstraintWeights_5:float=0.0364
    ConstraintWeights_6:float=0.0286
    ConstraintWeights_7:float=0.0487
    ConstraintWeights_8:float=0.0132
    ConstraintWeights_9:float=0.0088
   ) -> (Pose) {
    alphaBoolBlend {blendOption=Linear}
    boneToModify {boneName=dyn_strap_02_r}
    constraintSetup [
     {targetBone {boneName=dyn_coatTail02_r}}
     {targetBone {boneName=dyn_coatTail04_r}}
     {targetBone {boneName=dyn_coatTail03_r}}
     {targetBone {boneName=dyn_coatBack03_r}}
     {targetBone {boneName=dyn_splitUp_r}}
     {targetBone {boneName=dyn_coatBack04_r}}
     {targetBone {boneName=dyn_splitDown_r}}
     {targetBone {boneName=dyn_coatTail05_r}}
     {targetBone {boneName=dyn_coatBack02_r}}
     {targetBone {boneName=dyn_coatSplit03}}
    ]
    constraintWeights [1,1,1,1,1,1,1,1,1,1]
   }
   "Transform (Modify) Bone - Bone: dyn_jacketBack_03_l"(Alpha=0, Translation="0,0,0", Rotation="25,0,0.000", Scale="1,1,1") {
    rotationMode=BMM_Additive
    rotationSpace=BCS_BoneSpace
    alphaBoolBlend {blendOption=Linear}
    alphaScaleBiasClamp {bInterpResult=true, interpSpeedDecreasing=2, interpSpeedIncreasing=5}
    boneToModify {boneName=dyn_jacketBack_03_l}
    rotation {pitch=25}
   }
   n1 = "Transform (Modify) Bone - Bone: dyn_jacketFront_mstr_r"(Alpha=Get thigh_rot_up_r() -> (thigh_rot_up_r:double), Translation="0,0,3", Rotation="0,0,0", Scale="1,1,1") {
    rotationMode=BMM_Additive
    rotationSpace=BCS_BoneSpace
    translationMode=BMM_Additive
    translationSpace=BCS_BoneSpace
    alphaBoolBlend {blendOption=Linear}
    boneToModify {boneName=dyn_jacketFront_mstr_r}
    translation {z=3}
   }
   n2 = Get thigh_rot_otr_r() -> (thigh_rot_otr_r:double)
   n3 = "Transform (Modify) Bone - Bone: dyn_jacketFront_01_r"(ComponentPose=@n1, Alpha=@n2, Translation="0,0,0", Rotation="35,0,0", Scale="1,1,1") {
    rotationMode=BMM_Additive
    rotationSpace=BCS_BoneSpace
    translationMode=BMM_Additive
    translationSpace=BCS_BoneSpace
    alphaBoolBlend {blendOption=Linear}
    boneToModify {boneName=dyn_jacketFront_01_r}
    rotation {pitch=35}
   }
   n4 = "Transform (Modify) Bone - Bone: dyn_jacketFront_01_r"(ComponentPose=@n3, Alpha=@n2, Translation="0,0,2", Rotation="30,0,0", Scale="1,1,1") {
    rotationMode=BMM_Additive
    rotationSpace=BCS_BoneSpace
    translationMode=BMM_Additive
    translationSpace=BCS_BoneSpace
    alphaBoolBlend {blendOption=Linear}
    boneToModify {boneName=dyn_jacketFront_01_r}
    rotation {pitch=30}
    translation {z=2}
   }
   n5 = "Transform (Modify) Bone - Bone: dyn_jacketSide_02_r"(ComponentPose=@n4, Alpha=@n2, Translation="0,0,0", Rotation="30,0,0", Scale="1,1,1") {
    rotationMode=BMM_Additive
    rotationSpace=BCS_BoneSpace
    translationMode=BMM_Additive
    translationSpace=BCS_BoneSpace
    alphaBoolBlend {blendOption=Linear}
    boneToModify {boneName=dyn_jacketSide_02_r}
    rotation {pitch=30}
   }
   "Transform (Modify) Bone - Bone: dyn_coatFront01_Base_l"(ComponentPose=@n5, Alpha=1, Translation="-1,3,7", Rotation="5,10,15", Scale="1,1,1") {
    rotationMode=BMM_Additive
    rotationSpace=BCS_BoneSpace
    scaleMode=BMM_Additive
    scaleSpace=BCS_BoneSpace
    translationMode=BMM_Additive
    translationSpace=BCS_BoneSpace
    alphaBoolBlend {blendOption=Linear}
    boneToModify {boneName=dyn_coatFront01_Base_l}
    rotation=<rot 5 10 15>
    translation=<-1 3 7>
   }
   n6 = Get thigh_rot_up_l() -> (thigh_rot_up_l:double)
   "Transform (Modify) Bone - Bone: dyn_jacketFront_01_l"(Alpha=@n6, Translation="0,0,0", Rotation="10,-15.000,-35", Scale="1,1,1") {
    rotationMode=BMM_Additive
    rotationSpace=BCS_BoneSpace
    translationMode=BMM_Additive
    translationSpace=BCS_BoneSpace
    alphaBoolBlend {blendOption=Linear}
    boneToModify {boneName=dyn_jacketFront_01_l}
    rotation=<rot 10 -15 -35>
   }
   "Transform (Modify) Bone - Bone: dyn_jacketSide_01_l"(Alpha=@n6, Translation="0,-2,3", Rotation="0,-10,0", Scale="1,1,1") {
    rotationMode=BMM_Additive
    rotationSpace=BCS_BoneSpace
    translationMode=BMM_Additive
    translationSpace=BCS_BoneSpace
    alphaBoolBlend {blendOption=Linear}
    boneToModify {boneName=dyn_jacketSide_01_l}
    rotation {yaw=-10}
    translation {y=-2, z=3}
   }
   n7 = "Transform (Modify) Bone - Bone: dyn_jacketFront_02_l"(Alpha=@n6, Translation="0,0,0", Rotation="-25,0,0", Scale="1,1,1") {
    rotationMode=BMM_Additive
    rotationSpace=BCS_BoneSpace
    translationMode=BMM_Additive
    translationSpace=BCS_BoneSpace
    alphaBoolBlend {blendOption=Linear}
    boneToModify {boneName=dyn_jacketFront_02_l}
    rotation {pitch=-25}
   }
   n8 = "Transform (Modify) Bone - Bone: dyn_jacketSide_mstr_l"(ComponentPose=@n7, Alpha=1, Translation="-2,-2,0", Rotation="10,5,0", Scale="1,1,1") {
    rotationMode=BMM_Additive
    rotationSpace=BCS_BoneSpace
    scaleMode=BMM_Additive
    scaleSpace=BCS_BoneSpace
    translationMode=BMM_Additive
    translationSpace=BCS_BoneSpace
    alphaBoolBlend {blendOption=Linear}
    boneToModify {boneName=dyn_jacketSide_mstr_l}
    rotation {pitch=10, yaw=5}
    translation=<-2 -2>
   }
   n9 = Get thigh_rot_otr_l() -> (thigh_rot_otr_l:double)
   n10 = "Transform (Modify) Bone - Bone: dyn_jacketFront_01_l"(ComponentPose=@n8, Alpha=@n9, Translation="0,0,0", Rotation="45,0,0", Scale="1,1,1") {
    rotationMode=BMM_Additive
    rotationSpace=BCS_BoneSpace
    translationMode=BMM_Additive
    translationSpace=BCS_BoneSpace
    alphaBoolBlend {blendOption=Linear}
    boneToModify {boneName=dyn_jacketFront_01_l}
    rotation {pitch=45}
   }
   n11 = "Transform (Modify) Bone - Bone: dyn_jacketFront_02_l"(ComponentPose=@n10, Alpha=@n9, Translation="0,0,2", Rotation="30,0,0", Scale="1,1,1") {
    rotationMode=BMM_Additive
    rotationSpace=BCS_BoneSpace
    translationMode=BMM_Additive
    translationSpace=BCS_BoneSpace
    alphaBoolBlend {blendOption=Linear}
    boneToModify {boneName=dyn_jacketFront_02_l}
    rotation {pitch=30}
    translation {z=2}
   }
   "Transform (Modify) Bone - Bone: dyn_jacketSide_02_l"(ComponentPose=@n11, Alpha=@n9, Translation="0,0,0", Rotation="45,0,0", Scale="1,1,1") {
    rotationMode=BMM_Additive
    rotationSpace=BCS_BoneSpace
    translationMode=BMM_Additive
    translationSpace=BCS_BoneSpace
    alphaBoolBlend {blendOption=Linear}
    boneToModify {boneName=dyn_jacketSide_02_l}
    rotation {pitch=45}
   }
   n12 = "Constraint - dyn_strap_03_Base_r (PPPPPPP)"!(
    ComponentPose
    Alpha=1
    ConstraintWeights_0=0.0565
    ConstraintWeights_1:float=0.0066
    ConstraintWeights_2:float=0.0128
    ConstraintWeights_3:float=0.5865
    ConstraintWeights_4:float=0.174
    ConstraintWeights_5:float=0.1548
    ConstraintWeights_6:float=0.0088
   ) -> (Pose) {
    alphaBoolBlend {blendOption=Linear}
    boneToModify {boneName=dyn_strap_03_Base_r}
    constraintSetup [
     {targetBone {boneName=dyn_coatTail04_r}}
     {targetBone {boneName=dyn_coatTail03_r}}
     {targetBone {boneName=dyn_coatBack03_r}}
     {targetBone {boneName=dyn_coatTail05_r}}
     {targetBone {boneName=dyn_coatBack04_r}}
     {targetBone {boneName=dyn_splitDown_r}}
     {targetBone {boneName=dyn_splitUp_r}}
    ]
    constraintWeights [1,1,1,1,1,1,1]
   }
   n13 = "Constraint - dyn_strap_03_Base_l (PPPPPPP)"!(
    ComponentPose=@n12
    Alpha=1
    ConstraintWeights_0=0.5865
    ConstraintWeights_1:float=0.1548
    ConstraintWeights_2:float=0.174
    ConstraintWeights_3:float=0.0128
    ConstraintWeights_4:float=0.0088
    ConstraintWeights_5:float=0.0066
    ConstraintWeights_6:float=0.0565
   ) -> (Pose) {
    alphaBoolBlend {blendOption=Linear}
    boneToModify {boneName=dyn_strap_03_Base_l}
    constraintSetup [
     {targetBone {boneName=dyn_coatTail05_l}}
     {targetBone {boneName=dyn_splitDown_l}}
     {targetBone {boneName=dyn_coatBack04_l}}
     {targetBone {boneName=dyn_coatBack03_l}}
     {targetBone {boneName=dyn_splitUp_l}}
     {targetBone {boneName=dyn_coatTail03_l}}
     {targetBone {boneName=dyn_coatTail04_l}}
    ]
    constraintWeights [1,1,1,1,1,1,1]
   }
   RigidBody(ComponentPose=@n13, Alpha=0.4) {
    bClampLinearTranslationLimitToRefPose=true
    bOverrideWorldGravity=true
    overridePhysicsAsset=&/M_MED_ShiitakeShaolin_StrapPhysics
    simulationSpace=BaseBoneSpace
    alphaBoolBlend {blendOption=Linear}
    baseBoneRef {boneName=pelvis}
    componentAppliedLinearAccClamp=<800 800 600>
    componentLinearAccScale=<0.1 0.1 0.2>
    componentLinearVelScale=<0.05 0.05 0.1>
    overrideWorldGravity {z=-980}
   }
   n14 = Get Coat corr() -> (Coat corr:bool)
   n15 = Get Coat rotate corr() -> (Coat rotate corr:Rotator)
   n16 = Get Right coat transtate corr() -> (Right coat transtate corr:Vector)
   n17 = Get Left coat transtate corr() -> (Left coat transtate corr:Vector)
   n18 = Get thigh_rot_inr_r() -> (thigh_rot_inr_r:double)
   n19 = Look At - Bone: dyn_coatFront01_Base_r(ComponentPose=Local To Component(LocalPose=Use cached pose 'prerigid'()), Alpha=@n18) {
    alphaBoolBlend {blendOption=Linear}
    alphaScaleBiasClamp {bInterpResult=true}
    boneToModify {boneName=dyn_coatFront01_Base_r}
    lookAtLocation=<0 -12 5>
    lookAtTarget {boneReference {boneName=calf_r}}
    lookAt_Axis {axis=<1 0>}
    lookUp_Axis {axis {x=0}}
   }
   n20 = Look At - Bone: dyn_coatSide01_Base_r(ComponentPose=@n19, Alpha=@n18) {
    alphaBoolBlend {blendOption=Linear}
    alphaScaleBiasClamp {bInterpResult=true}
    boneToModify {boneName=dyn_coatSide01_Base_r}
    lookAtLocation {x=0, z=11}
    lookAtTarget {boneReference {boneName=calf_r}}
    lookAt_Axis {axis=<1 0>}
    lookUp_Axis {axis {x=0}}
   }
   n21 = Get thigh_rot_up_r()
   n22 = "Transform (Modify) Bone - Bone: dyn_coatFront01_r"(ComponentPose=@n20, Alpha=@n21, Translation="0,0,0", Rotation="5,-10,10.829", Scale="1,1,1") {
    rotationMode=BMM_Additive
    rotationSpace=BCS_BoneSpace
    translationMode=BMM_Additive
    translationSpace=BCS_BoneSpace
    alphaBoolBlend {blendOption=Linear}
    boneToModify {boneName=dyn_coatFront01_r}
    rotation=<rot 5 -10 10.829>
   }
   n23 = "Transform (Modify) Bone - Bone: dyn_coatSide01_r"(ComponentPose=@n22, Alpha=@n21, Translation="0,0,0", Rotation="5,0.000,0.000", Scale="1,1,1") {
    rotationMode=BMM_Additive
    rotationSpace=BCS_BoneSpace
    translationMode=BMM_Additive
    translationSpace=BCS_BoneSpace
    alphaBoolBlend {blendOption=Linear}
    boneToModify {boneName=dyn_coatSide01_r}
    rotation {pitch=5}
   }
   n24 = Get thigh_rot_inr_l() -> (thigh_rot_inr_l:double)
   n25 = Look At - Bone: dyn_coatFront01_Base_l(ComponentPose=@n23, Alpha=@n24) {
    alphaBoolBlend {blendOption=Linear}
    alphaScaleBiasClamp {bInterpResult=true}
    boneToModify {boneName=dyn_coatFront01_Base_l}
    lookAtLocation=<0 12 -5>
    lookAtTarget {boneReference {boneName=calf_l}}
    lookAt_Axis {axis=<-1 0>}
    lookUp_Axis {axis {x=0}}
   }
   n26 = Look At - Bone: dyn_coatSide01_Base_l(ComponentPose=@n25, Alpha=@n24) {
    alphaBoolBlend {blendOption=Linear}
    alphaScaleBiasClamp {bInterpResult=true}
    boneToModify {boneName=dyn_coatSide01_Base_l}
    lookAtLocation {x=0, z=-11}
    lookAtTarget {boneReference {boneName=calf_l}}
    lookAt_Axis {axis=<-1 0>}
    lookUp_Axis {axis {x=0}}
   }
   n27 = Get thigh_rot_up_l()
   n28 = "Transform (Modify) Bone - Bone: dyn_coatFront01_l"(ComponentPose=@n26, Alpha=@n27, Translation="0,0,0", Rotation="5,-10,-10.828", Scale="1,1,1") {
    rotationMode=BMM_Additive
    rotationSpace=BCS_BoneSpace
    translationMode=BMM_Additive
    translationSpace=BCS_BoneSpace
    alphaBoolBlend {blendOption=Linear}
    boneToModify {boneName=dyn_coatFront01_l}
    rotation=<rot 5 -10 -10.828>
   }
   n29 = "Transform (Modify) Bone - Bone: dyn_coatSide01_l"(ComponentPose=@n28, Alpha=@n27, Translation="0,0,0", Rotation="5,0,0", Scale="1,1,1") {
    rotationMode=BMM_Additive
    rotationSpace=BCS_BoneSpace
    translationMode=BMM_Additive
    translationSpace=BCS_BoneSpace
    alphaBoolBlend {blendOption=Linear}
    boneToModify {boneName=dyn_coatSide01_l}
    rotation {pitch=5}
   }
   n30 = Get bIsMoving2D()
   n31 = "Transform (Modify) Bone - Bone: dyn_coatBack01_l"!(ComponentPose=@n29, bAlphaBoolEnabled:bool=@n30, Translation="0,0,0", Rotation="10,-10,0", Scale="1,1,1") -> (Pose) {
    alphaInputType=Bool
    rotationMode=BMM_Additive
    rotationSpace=BCS_BoneSpace
    alphaBoolBlend {blendInTime=0.2, blendOption=Linear, blendOutTime=0.2}
    alphaScaleBiasClamp {bInterpResult=true, interpSpeedDecreasing=2, interpSpeedIncreasing=5}
    boneToModify {boneName=dyn_coatBack01_l}
    rotation {pitch=10, yaw=-10}
   }
   n32 = "Transform (Modify) Bone - Bone: dyn_coatBack01_r"!(ComponentPose=@n31, bAlphaBoolEnabled:bool=@n30, Translation="0,0,0", Rotation="10,-10,0.000", Scale="1,1,1") -> (Pose) {
    alphaInputType=Bool
    rotationMode=BMM_Additive
    rotationSpace=BCS_BoneSpace
    alphaBoolBlend {blendInTime=0.2, blendOption=Linear, blendOutTime=0.2}
    alphaScaleBiasClamp {bInterpResult=true, interpSpeedDecreasing=2, interpSpeedIncreasing=5}
    boneToModify {boneName=dyn_coatBack01_r}
    rotation {pitch=10, yaw=-10}
   }
   n33 = "Transform (Modify) Bone - Bone: dyn_coatTail01_l"!(ComponentPose=@n32, bAlphaBoolEnabled:bool=@n30, Translation="0,0,0", Rotation="5,-5,0", Scale="1,1,1") -> (Pose) {
    alphaInputType=Bool
    rotationMode=BMM_Additive
    rotationSpace=BCS_BoneSpace
    alphaBoolBlend {blendInTime=0.2, blendOption=Linear, blendOutTime=0.2}
    alphaScaleBiasClamp {bInterpResult=true, interpSpeedDecreasing=2, interpSpeedIncreasing=5}
    boneToModify {boneName=dyn_coatTail01_l}
    rotation {pitch=5, yaw=-5}
   }
   n34 = "Transform (Modify) Bone - Bone: dyn_strap_03_Base_l"!(ComponentPose=@n33, bAlphaBoolEnabled:bool=@n30, Translation="0,0,0", Rotation="5,-10,0", Scale="1,1,1") -> (Pose) {
    alphaInputType=Bool
    rotationMode=BMM_Additive
    rotationSpace=BCS_BoneSpace
    alphaBoolBlend {blendInTime=0.2, blendOption=Linear, blendOutTime=0.2}
    alphaScaleBiasClamp {bInterpResult=true, interpSpeedDecreasing=2, interpSpeedIncreasing=5}
    boneToModify {boneName=dyn_strap_03_Base_l}
    rotation {pitch=5, yaw=-10}
   }
   n35 = "Transform (Modify) Bone - Bone: dyn_strap_03_Base_r"!(ComponentPose=@n34, bAlphaBoolEnabled:bool=@n30, Translation="0,0,0", Rotation="5,-10,0.000", Scale="1,1,1") -> (Pose) {
    alphaInputType=Bool
    rotationMode=BMM_Additive
    rotationSpace=BCS_BoneSpace
    alphaBoolBlend {blendInTime=0.2, blendOption=Linear, blendOutTime=0.2}
    alphaScaleBiasClamp {bInterpResult=true, interpSpeedDecreasing=2, interpSpeedIncreasing=5}
    boneToModify {boneName=dyn_strap_03_Base_r}
    rotation {pitch=5, yaw=-10}
   }
   n36 = "Transform (Modify) Bone - Bone: dyn_coatTail01_r"!(ComponentPose=@n35, bAlphaBoolEnabled:bool=@n30, Translation="0,0,0", Rotation="5,-5,0.000", Scale="1,1,1") -> (Pose) {
    alphaInputType=Bool
    rotationMode=BMM_Additive
    rotationSpace=BCS_BoneSpace
    alphaBoolBlend {blendInTime=0.2, blendOption=Linear, blendOutTime=0.2}
    alphaScaleBiasClamp {bInterpResult=true, interpSpeedDecreasing=2, interpSpeedIncreasing=5}
    boneToModify {boneName=dyn_coatTail01_r}
    rotation {pitch=5, yaw=-5}
   }
   n37 = Get jumpingAlpha() -> (jumpingAlpha:double)
   n38 = "Transform (Modify) Bone - Bone: dyn_coatBack01_l"!(ComponentPose=@n36, Alpha:float=@n37, Translation="0,0,0", Rotation="15,-15,0.000", Scale="1,1,1") -> (Pose) {
    rotationMode=BMM_Additive
    rotationSpace=BCS_BoneSpace
    alphaBoolBlend {blendOption=Linear}
    alphaScaleBiasClamp {bInterpResult=true, interpSpeedDecreasing=2, interpSpeedIncreasing=5}
    boneToModify {boneName=dyn_coatBack01_l}
    rotation {pitch=15, yaw=-15}
   }
   n39 = "Transform (Modify) Bone - Bone: dyn_coatBack01_r"!(ComponentPose=@n38, Alpha:float=@n37, Translation="0,0,0", Rotation="15,-15,0.000", Scale="1,1,1") -> (Pose) {
    rotationMode=BMM_Additive
    rotationSpace=BCS_BoneSpace
    alphaBoolBlend {blendOption=Linear}
    alphaScaleBiasClamp {bInterpResult=true, interpSpeedDecreasing=2, interpSpeedIncreasing=5}
    boneToModify {boneName=dyn_coatBack01_r}
    rotation {pitch=15, yaw=-15}
   }
   n40 = "Transform (Modify) Bone - Bone: dyn_coatSide01_r"(ComponentPose=@n39, Alpha=@n37, Translation="0,0,0", Rotation="15,-15,0", Scale="1,1,1") {
    rotationMode=BMM_Additive
    rotationSpace=BCS_BoneSpace
    alphaBoolBlend {blendOption=Linear}
    alphaScaleBiasClamp {bInterpResult=true, interpSpeedDecreasing=2, interpSpeedIncreasing=5}
    boneToModify {boneName=dyn_coatSide01_r}
    rotation {pitch=15, yaw=-15}
   }
   n41 = "Transform (Modify) Bone - Bone: dyn_coatSide01_l"(ComponentPose=@n40, Alpha=@n37, Translation="0,0,0", Rotation="15,-15,0.000", Scale="1,1,1") {
    rotationMode=BMM_Additive
    rotationSpace=BCS_BoneSpace
    alphaBoolBlend {blendOption=Linear}
    alphaScaleBiasClamp {bInterpResult=true, interpSpeedDecreasing=2, interpSpeedIncreasing=5}
    boneToModify {boneName=dyn_coatSide01_l}
    rotation {pitch=15, yaw=-15}
   }
   n42 = "Transform (Modify) Bone - Bone: dyn_coatTail01_r"!(ComponentPose=@n41, Alpha:float=@n37, Translation="0,0,0", Rotation="0,-15,0", Scale="1,1,1") -> (Pose) {
    rotationMode=BMM_Additive
    rotationSpace=BCS_BoneSpace
    alphaBoolBlend {blendOption=Linear}
    alphaScaleBiasClamp {bInterpResult=true, interpSpeedDecreasing=2, interpSpeedIncreasing=5}
    boneToModify {boneName=dyn_coatTail01_r}
    rotation {yaw=-15}
   }
   n43 = "Transform (Modify) Bone - Bone: dyn_strap_03_Base_r"!(ComponentPose=@n42, Alpha:float=@n37, Translation="0,0,0", Rotation="0,-15,0", Scale="1,1,1") -> (Pose) {
    rotationMode=BMM_Additive
    rotationSpace=BCS_BoneSpace
    alphaBoolBlend {blendOption=Linear}
    alphaScaleBiasClamp {bInterpResult=true, interpSpeedDecreasing=2, interpSpeedIncreasing=5}
    boneToModify {boneName=dyn_strap_03_Base_r}
    rotation {yaw=-15}
   }
   n44 = "Transform (Modify) Bone - Bone: dyn_strap_03_Base_l"!(ComponentPose=@n43, Alpha:float=@n37, Translation="0,0,0", Rotation="0,-15,0", Scale="1,1,1") -> (Pose) {
    rotationMode=BMM_Additive
    rotationSpace=BCS_BoneSpace
    alphaBoolBlend {blendOption=Linear}
    alphaScaleBiasClamp {bInterpResult=true, interpSpeedDecreasing=2, interpSpeedIncreasing=5}
    boneToModify {boneName=dyn_strap_03_Base_l}
    rotation {yaw=-15}
   }
   n45 = "Transform (Modify) Bone - Bone: dyn_coatTail01_l"!(ComponentPose=@n44, Alpha:float=@n37, Translation="0,0,0", Rotation="0,-15,0", Scale="1,1,1") -> (Pose) {
    rotationMode=BMM_Additive
    rotationSpace=BCS_BoneSpace
    alphaBoolBlend {blendOption=Linear}
    alphaScaleBiasClamp {bInterpResult=true, interpSpeedDecreasing=2, interpSpeedIncreasing=5}
    boneToModify {boneName=dyn_coatTail01_l}
    rotation {yaw=-15}
   }
   n46 = Get skydive alpha() -> (skydive alpha:double)
   n47 = "Transform (Modify) Bone - Bone: dyn_coatBack01_Base_l"(ComponentPose=@n45, Alpha=@n46, Translation="0,0,0", Rotation="-15,-10,0.000", Scale="1,1,1") {
    rotationMode=BMM_Additive
    rotationSpace=BCS_BoneSpace
    alphaBoolBlend {blendOption=Linear}
    alphaScaleBiasClamp {bInterpResult=true, interpSpeedDecreasing=2, interpSpeedIncreasing=5}
    boneToModify {boneName=dyn_coatBack01_Base_l}
    rotation {pitch=-15, yaw=-10}
   }
   n48 = "Transform (Modify) Bone - Bone: dyn_coatBack03_l"(ComponentPose=@n47, Alpha=@n46, Translation="0,0,0", Rotation="-5,-10,0.000", Scale="1,1,1") {
    rotationMode=BMM_Additive
    rotationSpace=BCS_BoneSpace
    alphaBoolBlend {blendOption=Linear}
    alphaScaleBiasClamp {bInterpResult=true, interpSpeedDecreasing=2, interpSpeedIncreasing=5}
    boneToModify {boneName=dyn_coatBack03_l}
    rotation {pitch=-5, yaw=-10}
   }
   n49 = "Transform (Modify) Bone - Bone: dyn_coatBack01_Base_r"(ComponentPose=@n48, Alpha=@n46, Translation="0,0,0", Rotation="-15,-10,0.000", Scale="1,1,1") {
    rotationMode=BMM_Additive
    rotationSpace=BCS_BoneSpace
    alphaBoolBlend {blendOption=Linear}
    alphaScaleBiasClamp {bInterpResult=true, interpSpeedDecreasing=2, interpSpeedIncreasing=5}
    boneToModify {boneName=dyn_coatBack01_Base_r}
    rotation {pitch=-15, yaw=-10}
   }
   n50 = "Transform (Modify) Bone - Bone: dyn_coatTail01_r"(ComponentPose=@n49, Alpha=@n46, Translation="0,0,0", Rotation="15,-15,0", Scale="1,1,1") {
    rotationMode=BMM_Additive
    rotationSpace=BCS_BoneSpace
    alphaBoolBlend {blendOption=Linear}
    alphaScaleBiasClamp {bInterpResult=true, interpSpeedDecreasing=2, interpSpeedIncreasing=5}
    boneToModify {boneName=dyn_coatTail01_r}
    rotation {pitch=15, yaw=-15}
   }
   n51 = "Transform (Modify) Bone - Bone: dyn_coatTail01_l"(ComponentPose=@n50, Alpha=@n46, Translation="0,0,0", Rotation="15,-15,0.000", Scale="1,1,1") {
    rotationMode=BMM_Additive
    rotationSpace=BCS_BoneSpace
    alphaBoolBlend {blendOption=Linear}
    alphaScaleBiasClamp {bInterpResult=true, interpSpeedDecreasing=2, interpSpeedIncreasing=5}
    boneToModify {boneName=dyn_coatTail01_l}
    rotation {pitch=15, yaw=-15}
   }
   n52 = Get ParachuteAlpha() -> (ParachuteAlpha:double)
   n53 = "Transform (Modify) Bone - Bone: dyn_coatBack01_l"(ComponentPose=@n51, Alpha=@n52, Translation="0,0,0", Rotation="15,-10,0.000", Scale="1,1,1") {
    rotationMode=BMM_Additive
    rotationSpace=BCS_BoneSpace
    alphaBoolBlend {blendOption=Linear}
    alphaScaleBiasClamp {bInterpResult=true, interpSpeedDecreasing=2, interpSpeedIncreasing=5}
    boneToModify {boneName=dyn_coatBack01_l}
    rotation {pitch=15, yaw=-10}
   }
   n54 = "Transform (Modify) Bone - Bone: dyn_coatBack01_r"(ComponentPose=@n53, Alpha=@n52, Translation="0,0,0", Rotation="15,-10.000,0.000", Scale="1,1,1") {
    rotationMode=BMM_Additive
    rotationSpace=BCS_BoneSpace
    alphaBoolBlend {blendOption=Linear}
    alphaScaleBiasClamp {bInterpResult=true, interpSpeedDecreasing=2, interpSpeedIncreasing=5}
    boneToModify {boneName=dyn_coatBack01_r}
    rotation {pitch=15, yaw=-10}
   }
   n55 = "Transform (Modify) Bone - Bone: dyn_coatSide01_l"(ComponentPose=@n54, Alpha=@n52, Translation="0,0,0", Rotation="15,-15,0.000", Scale="1,1,1") {
    rotationMode=BMM_Additive
    rotationSpace=BCS_BoneSpace
    alphaBoolBlend {blendOption=Linear}
    alphaScaleBiasClamp {bInterpResult=true, interpSpeedDecreasing=2, interpSpeedIncreasing=5}
    boneToModify {boneName=dyn_coatSide01_l}
    rotation {pitch=15, yaw=-15}
   }
   n56 = "Transform (Modify) Bone - Bone: dyn_coatSide01_r"(ComponentPose=@n55, Alpha=@n52, Translation="0,0,0", Rotation="15,-15,0.000", Scale="1,1,1") {
    rotationMode=BMM_Additive
    rotationSpace=BCS_BoneSpace
    alphaBoolBlend {blendOption=Linear}
    alphaScaleBiasClamp {bInterpResult=true, interpSpeedDecreasing=2, interpSpeedIncreasing=5}
    boneToModify {boneName=dyn_coatSide01_r}
    rotation {pitch=15, yaw=-15}
   }
   n57 = "Transform (Modify) Bone - Bone: dyn_coatTail01_r"(ComponentPose=@n56, Alpha=@n52, Translation="0,0,0", Rotation="15,-15,0", Scale="1,1,1") {
    rotationMode=BMM_Additive
    rotationSpace=BCS_BoneSpace
    alphaBoolBlend {blendOption=Linear}
    alphaScaleBiasClamp {bInterpResult=true, interpSpeedDecreasing=2, interpSpeedIncreasing=5}
    boneToModify {boneName=dyn_coatTail01_r}
    rotation {pitch=15, yaw=-15}
   }
   n58 = "Transform (Modify) Bone - Bone: dyn_coatTail01_l"(ComponentPose=@n57, Alpha=@n52, Translation="0,0,0", Rotation="15,-15,0.000", Scale="1,1,1") {
    rotationMode=BMM_Additive
    rotationSpace=BCS_BoneSpace
    alphaBoolBlend {blendOption=Linear}
    alphaScaleBiasClamp {bInterpResult=true, interpSpeedDecreasing=2, interpSpeedIncreasing=5}
    boneToModify {boneName=dyn_coatTail01_l}
    rotation {pitch=15, yaw=-15}
   }
   n59 = "Transform (Modify) Bone - Bone: dyn_coatFront01_r"(ComponentPose=Local To Component(LocalPose=Use cached pose 'prerigid'()), Alpha=1, Translation="0,0,0", Rotation="0,-15,0", Scale="1,1,1") {
    rotationMode=BMM_Additive
    rotationSpace=BCS_BoneSpace
    translationMode=BMM_Additive
    translationSpace=BCS_BoneSpace
    alphaBoolBlend {blendOption=Linear}
    boneToModify {boneName=dyn_coatFront01_r}
    rotation {yaw=-15}
   }
   n60 = "Transform (Modify) Bone - Bone: dyn_coatSide01_r"(ComponentPose=@n59, Alpha=1, Translation="0,0,0", Rotation="0,-20,0", Scale="1,1,1") {
    rotationMode=BMM_Additive
    rotationSpace=BCS_BoneSpace
    translationMode=BMM_Additive
    translationSpace=BCS_BoneSpace
    alphaBoolBlend {blendOption=Linear}
    boneToModify {boneName=dyn_coatSide01_r}
    rotation {yaw=-20}
   }
   n61 = Get isNotCrouchRunning() -> (isNotCrouchRunning:double)
   n62 = "Transform (Modify) Bone - Bone: dyn_coatBack01_r"(ComponentPose=@n60, Alpha=@n61, Translation="0,0,0", Rotation="15,-10,0.000", Scale="1,1,1") {
    rotationMode=BMM_Additive
    rotationSpace=BCS_BoneSpace
    alphaBoolBlend {blendOption=Linear}
    alphaScaleBiasClamp {bInterpResult=true, interpSpeedDecreasing=5}
    boneToModify {boneName=dyn_coatBack01_r}
    rotation {pitch=15, yaw=-10}
   }
   n63 = "Transform (Modify) Bone - Bone: dyn_coatBack02_r"(ComponentPose=@n62, Alpha=@n61, Translation="0,0,1", Rotation="10,0,0.000", Scale="1,1,1") {
    rotationMode=BMM_Additive
    rotationSpace=BCS_BoneSpace
    alphaBoolBlend {blendOption=Linear}
    alphaScaleBiasClamp {bInterpResult=true, interpSpeedDecreasing=5}
    boneToModify {boneName=dyn_coatBack02_r}
    rotation {pitch=10}
    translation {z=1}
   }
   n64 = "Transform (Modify) Bone - Bone: dyn_coatBack03_r"(ComponentPose=@n63, Alpha=@n61, Translation="0,0,0", Rotation="5,0,0.000", Scale="1,1,1") {
    rotationMode=BMM_Additive
    rotationSpace=BCS_BoneSpace
    alphaBoolBlend {blendOption=Linear}
    alphaScaleBiasClamp {bInterpResult=true, interpSpeedDecreasing=5}
    boneToModify {boneName=dyn_coatBack03_r}
    rotation {pitch=5}
   }
   n65 = "Transform (Modify) Bone - Bone: dyn_coatBack04_r"(ComponentPose=@n64, Alpha=@n61, Translation="0,0,0", Rotation="-20,0,0.000", Scale="1,1,1") {
    rotationMode=BMM_Additive
    rotationSpace=BCS_BoneSpace
    alphaBoolBlend {blendOption=Linear}
    alphaScaleBiasClamp {bInterpResult=true, interpSpeedDecreasing=5}
    boneToModify {boneName=dyn_coatBack04_r}
    rotation {pitch=-20}
   }
   n66 = "Transform (Modify) Bone - Bone: dyn_coatBack01_l"(ComponentPose=@n65, Alpha=@n61, Translation="0,0,0", Rotation="10.000,-3.829,0.000", Scale="1,1,1") {
    rotationMode=BMM_Additive
    rotationSpace=BCS_BoneSpace
    alphaBoolBlend {blendOption=Linear}
    alphaScaleBiasClamp {bInterpResult=true, interpSpeedDecreasing=5}
    boneToModify {boneName=dyn_coatBack01_l}
    rotation {pitch=10, yaw=-3.829}
   }
   n67 = "Transform (Modify) Bone - Bone: dyn_coatBack02_l"(ComponentPose=@n66, Alpha=@n61, Translation="0,0,0", Rotation="15,0,0.000", Scale="1,1,1") {
    rotationMode=BMM_Additive
    rotationSpace=BCS_BoneSpace
    alphaBoolBlend {blendOption=Linear}
    alphaScaleBiasClamp {bInterpResult=true, interpSpeedDecreasing=5}
    boneToModify {boneName=dyn_coatBack02_l}
    rotation {pitch=15}
   }
   n68 = "Transform (Modify) Bone - Bone: dyn_coatBack03_l"(ComponentPose=@n67, Alpha=@n61, Translation="0,0,0", Rotation="30,0,0.000", Scale="1,1,1") {
    rotationMode=BMM_Additive
    rotationSpace=BCS_BoneSpace
    alphaBoolBlend {blendOption=Linear}
    alphaScaleBiasClamp {bInterpResult=true, interpSpeedDecreasing=5}
    boneToModify {boneName=dyn_coatBack03_l}
    rotation {pitch=30}
   }
   n69 = "Transform (Modify) Bone - Bone: dyn_coatBack04_l"(ComponentPose=@n68, Alpha=0, Translation="0,0,0", Rotation="5,0,0.000", Scale="1,1,1") {
    rotationMode=BMM_Additive
    rotationSpace=BCS_BoneSpace
    alphaBoolBlend {blendOption=Linear}
    alphaScaleBiasClamp {bInterpResult=true, interpSpeedDecreasing=5}
    boneToModify {boneName=dyn_coatBack04_l}
    rotation {pitch=5}
   }
   n70 = Get isNotCrouchRunning()
   n71 = "Transform (Modify) Bone - Bone: dyn_coatTail01_r"(ComponentPose=@n69, Alpha=@n70, Translation="0,0,0", Rotation="5,-10,0.000", Scale="1,1,1") {
    rotationMode=BMM_Additive
    rotationSpace=BCS_BoneSpace
    alphaBoolBlend {blendOption=Linear}
    alphaScaleBiasClamp {bInterpResult=true, interpSpeedDecreasing=5}
    boneToModify {boneName=dyn_coatTail01_r}
    rotation {pitch=5, yaw=-10}
   }
   n72 = "Transform (Modify) Bone - Bone: dyn_coatTail02_r"(ComponentPose=@n71, Alpha=@n70, Translation="0,0,1", Rotation="5,0,0.000", Scale="1,1,1") {
    rotationMode=BMM_Additive
    rotationSpace=BCS_BoneSpace
    alphaBoolBlend {blendOption=Linear}
    alphaScaleBiasClamp {bInterpResult=true, interpSpeedDecreasing=5}
    boneToModify {boneName=dyn_coatTail02_r}
    rotation {pitch=5}
    translation {z=1}
   }
   n73 = "Transform (Modify) Bone - Bone: dyn_coatTail03_r"(ComponentPose=@n72, Alpha=@n70, Translation="0,0,0", Rotation="5,0,0.000", Scale="1,1,1") {
    rotationMode=BMM_Additive
    rotationSpace=BCS_BoneSpace
    alphaBoolBlend {blendOption=Linear}
    alphaScaleBiasClamp {bInterpResult=true, interpSpeedDecreasing=5}
    boneToModify {boneName=dyn_coatTail03_r}
    rotation {pitch=5}
   }
   n74 = "Transform (Modify) Bone - Bone: dyn_coatTail04_r"(ComponentPose=@n73, Alpha=@n70, Translation="0,0,0", Rotation="-5,0,0.000", Scale="1,1,1") {
    rotationMode=BMM_Additive
    rotationSpace=BCS_BoneSpace
    alphaBoolBlend {blendOption=Linear}
    alphaScaleBiasClamp {bInterpResult=true, interpSpeedDecreasing=5}
    boneToModify {boneName=dyn_coatTail04_r}
    rotation {pitch=-5}
   }
   n75 = "Transform (Modify) Bone - Bone: dyn_coatTail01_l"(ComponentPose=@n74, Alpha=@n70, Translation="0,0,0", Rotation="0,-3.829,0.000", Scale="1,1,1") {
    rotationMode=BMM_Additive
    rotationSpace=BCS_BoneSpace
    alphaBoolBlend {blendOption=Linear}
    alphaScaleBiasClamp {bInterpResult=true, interpSpeedDecreasing=5}
    boneToModify {boneName=dyn_coatTail01_l}
    rotation {yaw=-3.829}
   }
   n76 = "Transform (Modify) Bone - Bone: dyn_coatTail02_l"(ComponentPose=@n75, Alpha=@n70, Translation="0,0,0", Rotation="5,0,0.000", Scale="1,1,1") {
    rotationMode=BMM_Additive
    rotationSpace=BCS_BoneSpace
    alphaBoolBlend {blendOption=Linear}
    alphaScaleBiasClamp {bInterpResult=true, interpSpeedDecreasing=5}
    boneToModify {boneName=dyn_coatTail02_l}
    rotation {pitch=5}
   }
   n77 = "Transform (Modify) Bone - Bone: dyn_coatTail03_l"(ComponentPose=@n76, Alpha=@n70, Translation="0,0,0", Rotation="10,0,0.000", Scale="1,1,1") {
    rotationMode=BMM_Additive
    rotationSpace=BCS_BoneSpace
    alphaBoolBlend {blendOption=Linear}
    alphaScaleBiasClamp {bInterpResult=true, interpSpeedDecreasing=5}
    boneToModify {boneName=dyn_coatTail03_l}
    rotation {pitch=10}
   }
   n78 = "Transform (Modify) Bone - Bone: dyn_coatTail04_l"(ComponentPose=@n77, Alpha=@n70, Translation="0,0,0", Rotation="5,0,0.000", Scale="1,1,1") {
    rotationMode=BMM_Additive
    rotationSpace=BCS_BoneSpace
    alphaBoolBlend {blendOption=Linear}
    alphaScaleBiasClamp {bInterpResult=true, interpSpeedDecreasing=5}
    boneToModify {boneName=dyn_coatTail04_l}
    rotation {pitch=5}
   }
   n79 = Blend Poses by bool(BlendPose_0=Component To Local(ComponentPose=@n78), BlendPose_1=Component To Local(ComponentPose=@n58), bActiveValue=Get bIsCrouching()) {blendType=Linear}
   n80 = Get Is Sliding Float() -> (Is Sliding Float:double)
   n81 = "Transform (Modify) Bone - Bone: dyn_coatBack01_Base_r"(ComponentPose=Local To Component(LocalPose=Use cached pose 'prerigid'()), Alpha=@n80, Translation="0,1,0", Rotation="0,-40,0", Scale="1,1,1") {
    rotationMode=BMM_Additive
    rotationSpace=BCS_BoneSpace
    translationMode=BMM_Additive
    translationSpace=BCS_BoneSpace
    alphaBoolBlend {blendOption=Linear}
    alphaScaleBiasClamp {bInterpResult=true, interpSpeedDecreasing=5}
    boneToModify {boneName=dyn_coatBack01_Base_r}
    rotation {yaw=-40}
    translation {y=1}
   }
   n82 = "Transform (Modify) Bone - Bone: dyn_coatBack01_Base_l"(ComponentPose=@n81, Alpha=@n80, Translation="0,1,0", Rotation="0,-40,0", Scale="1,1,1") {
    rotationMode=BMM_Additive
    rotationSpace=BCS_BoneSpace
    translationMode=BMM_Additive
    translationSpace=BCS_BoneSpace
    alphaBoolBlend {blendOption=Linear}
    alphaScaleBiasClamp {bInterpResult=true, interpSpeedDecreasing=5}
    boneToModify {boneName=dyn_coatBack01_Base_l}
    rotation {yaw=-40}
    translation {y=1}
   }
   n83 = "Transform (Modify) Bone - Bone: dyn_coatTail01_Base_r"(ComponentPose=@n82, Alpha=@n80, Translation="0,1,0", Rotation="0,-40,0", Scale="1,1,1") {
    rotationMode=BMM_Additive
    rotationSpace=BCS_BoneSpace
    translationMode=BMM_Additive
    translationSpace=BCS_BoneSpace
    alphaBoolBlend {blendOption=Linear}
    alphaScaleBiasClamp {bInterpResult=true, interpSpeedDecreasing=5}
    boneToModify {boneName=dyn_coatTail01_Base_r}
    rotation {yaw=-40}
    translation {y=1}
   }
   n84 = "Transform (Modify) Bone - Bone: dyn_coatTail01_Base_l"(ComponentPose=@n83, Alpha=@n80, Translation="0,1,0", Rotation="0,-40,0", Scale="1,1,1") {
    rotationMode=BMM_Additive
    rotationSpace=BCS_BoneSpace
    translationMode=BMM_Additive
    translationSpace=BCS_BoneSpace
    alphaBoolBlend {blendOption=Linear}
    alphaScaleBiasClamp {bInterpResult=true, interpSpeedDecreasing=5}
    boneToModify {boneName=dyn_coatTail01_Base_l}
    rotation {yaw=-40}
    translation {y=1}
   }
   n85 = RigidBody!(ComponentPose=@n84, Alpha, OverrideWorldGravity:Vector=Get GravityOverride(), ExternalForce) -> (Pose) {
    bClampLinearTranslationLimitToRefPose=true
    bOverrideWorldGravity=true
    overridePhysicsAsset=&/M_MED_Shiitake_Shaolin_Physics
    simulationSpace=BaseBoneSpace
    alphaBoolBlend {blendOption=Linear}
    baseBoneRef {boneName=pelvis}
    componentAppliedLinearAccClamp=<4000 4000 4000>
    componentLinearAccScale=<0.5 0.5 0.5>
    componentLinearVelScale=<0.3 0.3 0.3>
    simSpaceSettings {
     maxAngularAcceleration=200
     maxAngularVelocity=5
     maxLinearAcceleration=1000
     maxLinearVelocity=300
     velocityScaleZ=0
     worldAlpha=1
     externalLinearDragV=<1 1>
    }
   }
   n86 = "Transform (Modify) Bone - Bone: dyn_coatFront01_Base_l"(ComponentPose=@n85, Alpha=Get thigh_rot_up_l(), Translation="3,2.5,0", Rotation="0,0,0", Scale="1,1,1") {
    rotationMode=BMM_Additive
    rotationSpace=BCS_BoneSpace
    translationMode=BMM_Additive
    translationSpace=BCS_BoneSpace
    alphaBoolBlend {blendOption=Linear}
    boneToModify {boneName=dyn_coatFront01_Base_l}
    translation=<3 2.5>
   }
   n87 = RigidBody(ComponentPose=Local To Component(LocalPose=Use cached pose 'prerigid'()), OverrideWorldGravity=Get GravityOverride()) {
    bClampLinearTranslationLimitToRefPose=true
    bOverrideWorldGravity=true
    overridePhysicsAsset=&/M_MED_Shiitake_Shaolin_Falling_Physics
    simulationSpace=BaseBoneSpace
    alphaBoolBlend {blendOption=Linear}
    baseBoneRef {boneName=pelvis}
    componentAppliedLinearAccClamp=<4000 4000 4000>
    componentLinearAccScale=<0.5 0.5 0.5>
    componentLinearVelScale=<0.3 0.3 0.3>
    simSpaceSettings {
     maxAngularAcceleration=200
     maxAngularVelocity=5
     maxLinearAcceleration=1000
     maxLinearVelocity=300
     velocityScaleZ=0
     worldAlpha=1
     externalLinearDragV=<1 1>
    }
   }
   n88 = RigidBody(ComponentPose=Local To Component(LocalPose=Use cached pose 'prerigid'()), OverrideWorldGravity=Get GravityOverride()) {
    bClampLinearTranslationLimitToRefPose=true
    bOverrideWorldGravity=true
    overridePhysicsAsset=&M_MED_Shiitake_Shaolin_Physics
    simulationSpace=BaseBoneSpace
    alphaBoolBlend {blendOption=Linear}
    baseBoneRef {boneName=pelvis}
    componentAppliedLinearAccClamp=<4000 4000 4000>
    componentLinearAccScale=<0.5 0.5 0.5>
    componentLinearVelScale=<0.3 0.3 0.3>
    simSpaceSettings {
     maxAngularAcceleration=200
     maxAngularVelocity=5
     maxLinearAcceleration=1000
     maxLinearVelocity=300
     velocityScaleZ=0
     worldAlpha=1
     externalLinearDragV=<1 1>
    }
   }
   n89 = Blend Poses by bool(
    BlendPose_0=Component To Local(ComponentPose=@n88)
    BlendPose_1=Blend Poses by bool(
     BlendPose_0=Component To Local(ComponentPose=@n87)
     BlendPose_1=Component To Local(ComponentPose=@n86)
     BlendTime_0=0.5
     BlendTime_1=0.8
     bActiveValue=Get bIsFalling()
    ) {blendType=Linear}
    bActiveValue=Get bIsDBNOCarried()
   ) {blendType=Linear}
   n90 = Get bRigidBodyEnabled()
   n91 = "Transform (Modify) Bone - Bone: dyn_coatFront01_Base_l"!(
    ComponentPose=Local To Component(LocalPose=Blend Poses by bool(BlendPose_0=@n89, BlendPose_1=@n79, bActiveValue=@n90) {blendType=Linear})
    bAlphaBoolEnabled:bool=@n14
    Translation=@n17
    Rotation=@n15
    Scale="1,1,1"
   ) -> (Pose) {
    alphaInputType=Bool
    rotationMode=BMM_Additive
    rotationSpace=BCS_BoneSpace
    translationMode=BMM_Additive
    translationSpace=BCS_BoneSpace
    alphaBoolBlend {blendInTime=0.4, blendOption=CircularIn, blendOutTime=0.2}
    boneToModify {boneName=dyn_coatFront01_Base_l}
   }
   n92 = "Transform (Modify) Bone - Bone: dyn_coatSide01_Base_l"!(ComponentPose=@n91, bAlphaBoolEnabled:bool=@n14, Translation=@n17, Rotation=@n15, Scale="1,1,1") -> (Pose) {
    alphaInputType=Bool
    rotationMode=BMM_Additive
    rotationSpace=BCS_BoneSpace
    translationMode=BMM_Additive
    translationSpace=BCS_BoneSpace
    alphaBoolBlend {blendInTime=0.4, blendOption=CircularIn, blendOutTime=0.2}
    boneToModify {boneName=dyn_coatSide01_Base_l}
   }
   n93 = "Transform (Modify) Bone - Bone: dyn_coatBack01_Base_l"!(ComponentPose=@n92, bAlphaBoolEnabled:bool=@n14, Translation=@n17, Rotation=@n15, Scale="1,1,1") -> (Pose) {
    alphaInputType=Bool
    rotationMode=BMM_Additive
    rotationSpace=BCS_BoneSpace
    translationMode=BMM_Additive
    translationSpace=BCS_BoneSpace
    alphaBoolBlend {blendInTime=0.4, blendOption=CircularIn, blendOutTime=0.2}
    boneToModify {boneName=dyn_coatBack01_Base_l}
   }
   n94 = "Transform (Modify) Bone - Bone: dyn_coatFront01_Base_r"!(ComponentPose=@n93, bAlphaBoolEnabled:bool=@n14, Translation=@n16, Rotation=@n15, Scale="1,1,1") -> (Pose) {
    alphaInputType=Bool
    rotationMode=BMM_Additive
    rotationSpace=BCS_BoneSpace
    translationMode=BMM_Additive
    translationSpace=BCS_BoneSpace
    alphaBoolBlend {blendInTime=0.4, blendOption=CircularIn, blendOutTime=0.2}
    boneToModify {boneName=dyn_coatFront01_Base_r}
   }
   n95 = "Transform (Modify) Bone - Bone: dyn_coatSide01_Base_r"!(ComponentPose=@n94, bAlphaBoolEnabled:bool=@n14, Translation=@n16, Rotation=@n15, Scale="1,1,1") -> (Pose) {
    alphaInputType=Bool
    rotationMode=BMM_Additive
    rotationSpace=BCS_BoneSpace
    translationMode=BMM_Additive
    translationSpace=BCS_BoneSpace
    alphaBoolBlend {blendInTime=0.4, blendOption=CircularIn, blendOutTime=0.2}
    boneToModify {boneName=dyn_coatSide01_Base_r}
   }
   n96 = "Transform (Modify) Bone - Bone: dyn_coatBack01_Base_r"!(ComponentPose=@n95, bAlphaBoolEnabled:bool=@n14, Translation=@n16, Rotation=@n15, Scale="1,1,1") -> (Pose) {
    alphaInputType=Bool
    rotationMode=BMM_Additive
    rotationSpace=BCS_BoneSpace
    translationMode=BMM_Additive
    translationSpace=BCS_BoneSpace
    alphaBoolBlend {blendInTime=0.4, blendOption=CircularIn, blendOutTime=0.2}
    boneToModify {boneName=dyn_coatBack01_Base_r}
   }
   n97 = "Transform (Modify) Bone - Bone: dyn_coatTail02_l"!(ComponentPose=@n96, bAlphaBoolEnabled:bool=@n14, Translation="0,0,0", Rotation="0,-2,0", Scale="1,1,1") -> (Pose) {
    alphaInputType=Bool
    rotationMode=BMM_Additive
    rotationSpace=BCS_BoneSpace
    translationMode=BMM_Additive
    translationSpace=BCS_BoneSpace
    alphaBoolBlend {blendInTime=0.4, blendOption=CircularIn, blendOutTime=0.2}
    boneToModify {boneName=dyn_coatTail02_l}
    rotation {yaw=-2}
   }
   n98 = "Transform (Modify) Bone - Bone: dyn_coatTail02_r"!(ComponentPose=@n97, bAlphaBoolEnabled:bool=@n14, Translation="0,0,0", Rotation="0,-2,0", Scale="1,1,1") -> (Pose) {
    alphaInputType=Bool
    rotationMode=BMM_Additive
    rotationSpace=BCS_BoneSpace
    translationMode=BMM_Additive
    translationSpace=BCS_BoneSpace
    alphaBoolBlend {blendInTime=0.4, blendOption=CircularIn, blendOutTime=0.2}
    boneToModify {boneName=dyn_coatTail02_r}
    rotation {yaw=-2}
   }
   n99 = "Transform (Modify) Bone - Bone: dyn_coatSplit01_Base"!(ComponentPose=@n98, bAlphaBoolEnabled:bool=@n14, Translation="0.4,0,0", Rotation="0,2,0", Scale="1,1,1") -> (Pose) {
    alphaInputType=Bool
    rotationMode=BMM_Additive
    rotationSpace=BCS_BoneSpace
    translationMode=BMM_Additive
    translationSpace=BCS_BoneSpace
    alphaBoolBlend {blendInTime=0.4, blendOption=CircularIn, blendOutTime=0.2}
    boneToModify {boneName=dyn_coatSplit01_Base}
    rotation {yaw=2}
    translation {x=0.4}
   }
   n100 = Anim Dynamics - Bone: dyn_necklaceMain(ComponentPose=@n99, Alpha=1, ExternalForce="0,0,0") {
    angularDampingOverride=0.998
    angularSpringConstant=70
    bAngularSpring=true
    bGravityOverrideInSimSpace=true
    bOverrideAngularDamping=true
    bUseGravityOverride=true
    bUseSphericalLimits=true
    simulationSpace=RootRelative
    alphaBoolBlend {blendOption=Linear}
    boundBone {boneName=dyn_necklaceMain}
    componentAppliedLinearAccClamp=<3000 3000 3000>
    componentLinearAccScale=<0.3 0.3 0.3>
    componentLinearVelScale=<0.3 0.3 0.3>
    gravityOverride {z=-980}
    physicsBodyDefinitions [{
      collisionType=InnerSphere
      sphereCollisionRadius=1
      boundBone {boneName=dyn_necklaceMain}
      boxExtents=<7 12 12>
      constraintSetup {twistAxis=AxisZ, angularLimitsMax=<0 0 0>, angularLimitsMin=<0 0 -15>, angularTarget=<0 0 0>}
      localJointOffset=<2 0 0>
     }]
    planarLimits [{drivingBone {boneName=spine_05}, planeTransform {location=<0 0 2>, rotation=<rot 0 0 180>}}]
    relativeSpaceBone {boneName=root}
    sphericalLimits [{limitRadius=19.5, drivingBone {boneName=spine_04}, sphereLocalOffset=<0 -8 0>}]
   }
   n101 = Anim Dynamics - Bone: dyn_necklacePendant(ComponentPose=@n100, Alpha=1, ExternalForce="0,0,0") {
    angularDampingOverride=0.998
    angularSpringConstant=70
    bAngularSpring=true
    bGravityOverrideInSimSpace=true
    bOverrideAngularDamping=true
    bUseGravityOverride=true
    bUsePlanarLimit=false
    bUseSphericalLimits=true
    simulationSpace=RootRelative
    alphaBoolBlend {blendOption=Linear}
    boundBone {boneName=dyn_necklacePendant}
    componentAppliedLinearAccClamp=<3000 3000 3000>
    componentLinearAccScale=<0.3 0.3 0.3>
    componentLinearVelScale=<0.3 0.3 0.3>
    gravityOverride {z=-980}
    physicsBodyDefinitions [{
      collisionType=InnerSphere
      sphereCollisionRadius=1
      boundBone {boneName=dyn_necklacePendant}
      boxExtents=<7 12 12>
      constraintSetup {twistAxis=AxisZ, angularLimitsMax=<0 5 0>, angularLimitsMin=<0 -5 -35>, angularTarget=<0 0 0>}
      localJointOffset=<2 0 0>
     }]
    relativeSpaceBone {boneName=root}
    sphericalLimits [{limitRadius=19.5, drivingBone {boneName=spine_04}, sphereLocalOffset=<0 -9 0>}]
   }
   n102 = Look At - Bone: dyn_necklaceMain(ComponentPose=@n101, Alpha=1) {
    alphaBoolBlend {blendOption=Linear}
    boneToModify {boneName=dyn_necklaceMain}
    lookAtLocation {x=0}
    lookAtTarget {boneReference {boneName=dyn_necklacePendant}}
    lookAt_Axis {axis=<1 0>}
   }
   Output Pose(Result=Component To Local(ComponentPose=@n102)) {layerGroup=None}
   n103 = Control Rig(Source=Copy Pose From Mesh() {bUseAttachedParent=true}) {
    alphaBoolBlend {blendOption=Linear}
    controlRigAssetReference {blueprintRigClass=&/M_MED_ShiitakeShaolin_Straps_CtrlRig.M_MED_ShiitakeShaolin_Straps_CtrlRig_C}
   }
   n104 = Get skydive alpha()
   n105 = "Transform (Modify) Bone - Bone: dyn_coatBack01_Base_r"!(ComponentPose=Local To Component(LocalPose=@n103), Alpha:float=@n104, Translation="0,-5,0", Rotation="0,0,0", Scale="1,1,1") -> (Pose) {translationMode=BMM_Additive, alphaBoolBlend {blendOption=Linear}, boneToModify {boneName=dyn_coatBack01_Base_r}, translation {y=-5}}
   n106 = "Transform (Modify) Bone - Bone: dyn_coatBack01_Base_l"!(ComponentPose=@n105, Alpha:float=@n104, Translation="0,-5,0", Rotation="0,0,0", Scale="1,1,1") -> (Pose) {translationMode=BMM_Additive, alphaBoolBlend {blendOption=Linear}, boneToModify {boneName=dyn_coatBack01_Base_l}, translation {y=-5}}
   n107 = "Transform (Modify) Bone - Bone: dyn_coatSide01_Base_r"!(ComponentPose=@n106, Alpha:float=@n104, Translation="-3,-5,0", Rotation="0,0,0", Scale="1,1,1") -> (Pose) {translationMode=BMM_Additive, alphaBoolBlend {blendOption=Linear}, boneToModify {boneName=dyn_coatSide01_Base_r}, translation=<-3 -5>}
   n108 = "Transform (Modify) Bone - Bone: dyn_coatSide01_Base_l"!(ComponentPose=@n107, Alpha:float=@n104, Translation="3,-5,0", Rotation="0,0,0", Scale="1,1,1") -> (Pose) {translationMode=BMM_Additive, alphaBoolBlend {blendOption=Linear}, boneToModify {boneName=dyn_coatSide01_Base_l}, translation=<3 -5>}
   n109 = Get crouchAlpha() -> (crouchAlpha:double)
   n110 = "Transform (Modify) Bone - Bone: dyn_coatBack01_Base_r"(ComponentPose=@n108, Alpha=@n109, Translation="0,0,0", Rotation="0.000,-22,0.000", Scale="1,1,1") {
    rotationMode=BMM_Additive
    rotationSpace=BCS_BoneSpace
    alphaBoolBlend {blendOption=Linear}
    boneToModify {boneName=dyn_coatBack01_Base_r}
    rotation {yaw=-22}
   }
   n111 = "Transform (Modify) Bone - Bone: dyn_coatSide01_Base_r"(ComponentPose=@n110, Alpha=@n109, Translation="0,0,0", Rotation="0.000,-22,0.000", Scale="1,1,1") {
    rotationMode=BMM_Additive
    rotationSpace=BCS_BoneSpace
    alphaBoolBlend {blendOption=Linear}
    boneToModify {boneName=dyn_coatSide01_Base_r}
    rotation {yaw=-22}
   }
   n112 = "Transform (Modify) Bone - Bone: dyn_coatTail01_Base_r"(ComponentPose=@n111, Alpha=@n109, Translation="0,0,0", Rotation="0.000,-22,0.000", Scale="1,1,1") {
    rotationMode=BMM_Additive
    rotationSpace=BCS_BoneSpace
    alphaBoolBlend {blendOption=Linear}
    boneToModify {boneName=dyn_coatTail01_Base_r}
    rotation {yaw=-22}
   }
   n113 = Get thigh_rot_r()
   n114 = "Transform (Modify) Bone - Bone: dyn_coatFront01_Base_r"!(ComponentPose=@n112, Alpha:float=@n113, Translation="-5,0,5", Rotation="0,-25,0", Scale="1,1,1") -> (Pose) {
    rotationMode=BMM_Additive
    rotationSpace=BCS_BoneSpace
    scaleMode=BMM_Additive
    scaleSpace=BCS_BoneSpace
    translationMode=BMM_Additive
    translationSpace=BCS_BoneSpace
    alphaBoolBlend {blendOption=Linear}
    boneToModify {boneName=dyn_coatFront01_Base_r}
    rotation {yaw=-25}
    translation {x=-5, z=5}
   }
   n115 = "Transform (Modify) Bone - Bone: dyn_coatSide01_Base_r"(ComponentPose=@n114, Alpha=@n113, Translation="-2,-3,0", Rotation="0,-5,0", Scale="1,1,1") {
    rotationMode=BMM_Additive
    rotationSpace=BCS_BoneSpace
    scaleMode=BMM_Additive
    scaleSpace=BCS_BoneSpace
    translationMode=BMM_Additive
    translationSpace=BCS_BoneSpace
    alphaBoolBlend {blendOption=Linear}
    boneToModify {boneName=dyn_coatSide01_Base_r}
    rotation {yaw=-5}
    translation=<-2 -3>
   }
   n116 = "Transform (Modify) Bone - Bone: dyn_coatBack01_Base_r"(ComponentPose=@n115, Alpha=1, Translation="0,-3,-1", Rotation="0,-5,0", Scale="1,1,1") {
    rotationMode=BMM_Additive
    rotationSpace=BCS_BoneSpace
    scaleMode=BMM_Additive
    scaleSpace=BCS_BoneSpace
    translationMode=BMM_Additive
    translationSpace=BCS_BoneSpace
    alphaBoolBlend {blendOption=Linear}
    boneToModify {boneName=dyn_coatBack01_Base_r}
    rotation {yaw=-5}
    translation {y=-3, z=-1}
   }
   Save cached pose 'prerigid'(Pose=Component To Local(ComponentPose=@n116)) {cacheName=prerigid}
   n117 = Get BwdMovement() -> (BwdMovement:double)
   n118 = Trail controller - Bone: dyn_coatBack05_r!(ComponentPose, Alpha=@n117, FakeVelocity:Vector="0,0,0") -> (Pose) {
    bActorSpaceFakeVel=true
    bShowTrailLocation=true
    chainLength=4
    debugLifeTime=0.1
    maxDeltaTime=0.0333
    relaxationSpeedScale=3
    alphaScaleBiasClamp {bInterpResult=true}
    baseJoint {boneName=root}
    planarLimits [
     {drivingBone {boneName=root}, planeTransform {location=<0 0 10>, rotation=<rot 0 0 0>, scale=<1 1 1>}}
     {drivingBone {boneName=spine_03}, planeTransform {location=<0 -20 0>, rotation=<rot 0 15 -89.9999>, scale=<1 1 1>}}
    ]
    rotationLimits [{limitMax=<180 180 180>}, {limitMax=<180 180 180>}, {limitMax=<180 180 180>}, {limitMax=<180 180 180>}]
    rotationOffsets [<0 0 0>, <0 0 0>, <0 0 0>, <0 0 0>]
    trailBone {boneName=dyn_coatBack05_r}
    trailRelaxationSpeed {editorCurveData {keys [{arriveTangent=-72.0739, interpMode=RCIM_Cubic, leaveTangent=-72.0739, tangentMode=RCTM_User, time=0, value=35}, <at 1 5 RCIM_Cubic>]}}
   }
   Trail controller - Bone: dyn_coatBack05_l!(ComponentPose=@n118, Alpha=@n117, FakeVelocity:Vector="0,0,0") -> (Pose) {
    bActorSpaceFakeVel=true
    bShowTrailLocation=true
    chainLength=4
    debugLifeTime=0.1
    maxDeltaTime=0.0333
    relaxationSpeedScale=3
    alphaScaleBiasClamp {bInterpResult=true}
    baseJoint {boneName=root}
    planarLimits [
     {drivingBone {boneName=root}, planeTransform {location=<0 0 10>, rotation=<rot 0 0 0>, scale=<1 1 1>}}
     {drivingBone {boneName=spine_03}, planeTransform {location=<0 -20 0>, rotation=<rot 0 15 -89.9999>, scale=<1 1 1>}}
    ]
    rotationLimits [{limitMax=<180 180 180>}, {limitMax=<180 180 180>}, {limitMax=<180 180 180>}, {limitMax=<180 180 180>}]
    rotationOffsets [<0 0 0>, <0 0 0>, <0 0 0>, <0 0 0>]
    trailBone {boneName=dyn_coatBack05_l}
    trailRelaxationSpeed {editorCurveData {keys [{arriveTangent=-72.0739, interpMode=RCIM_Cubic, leaveTangent=-72.0739, tangentMode=RCTM_User, time=0, value=35}, <at 1 5 RCIM_Cubic>]}}
   }
   n119 = Get LeftMovement() -> (LeftMovement:double)
   n120 = Trail controller - Bone: dyn_coatFront06_r!(ComponentPose, Alpha=@n119, FakeVelocity:Vector="0,0,0") -> (Pose) {
    bActorSpaceFakeVel=true
    bShowTrailLocation=true
    chainLength=5
    debugLifeTime=0.1
    maxDeltaTime=0.0333
    relaxationSpeedScale=3
    alphaScaleBiasClamp {bInterpResult=true}
    baseJoint {boneName=root}
    planarLimits [
     {drivingBone {boneName=root}, planeTransform {location=<0 0 10>, rotation=<rot 0 0 0>, scale=<1 1 1>}}
     {drivingBone {boneName=spine_03}, planeTransform {location=<0 -20 0>, rotation=<rot 0 15 -89.9999>, scale=<1 1 1>}}
    ]
    rotationLimits [{limitMax=<180 180 180>}, {limitMax=<180 180 180>}, {limitMax=<180 180 180>}, {limitMax=<180 180 180>}, {limitMax=<180 180 180>}]
    rotationOffsets [<0 0 0>, <0 0 0>, <0 0 0>, <0 0 0>, <0 0 0>]
    trailBone {boneName=dyn_coatFront06_r}
    trailRelaxationSpeed {editorCurveData {keys [{arriveTangent=-72.0739, interpMode=RCIM_Cubic, leaveTangent=-72.0739, tangentMode=RCTM_User, time=0, value=35}, <at 1 5 RCIM_Cubic>]}}
   }
   n121 = Trail controller - Bone: dyn_coatSide05_r!(ComponentPose=@n120, Alpha=@n119, FakeVelocity:Vector="0,0,0") -> (Pose) {
    bActorSpaceFakeVel=true
    bShowTrailLocation=true
    chainLength=4
    debugLifeTime=0.1
    maxDeltaTime=0.0333
    relaxationSpeedScale=3
    alphaScaleBiasClamp {bInterpResult=true}
    baseJoint {boneName=root}
    planarLimits [
     {drivingBone {boneName=root}, planeTransform {location=<0 0 10>, rotation=<rot 0 0 0>, scale=<1 1 1>}}
     {drivingBone {boneName=spine_03}, planeTransform {location=<0 -20 0>, rotation=<rot 0 15 -89.9999>, scale=<1 1 1>}}
    ]
    rotationLimits [{limitMax=<180 180 180>}, {limitMax=<180 180 180>}, {limitMax=<180 180 180>}, {limitMax=<180 180 180>}]
    rotationOffsets [<0 0 0>, <0 0 0>, <0 0 0>, <0 0 0>]
    trailBone {boneName=dyn_coatSide05_r}
    trailRelaxationSpeed {editorCurveData {keys [{arriveTangent=-72.0739, interpMode=RCIM_Cubic, leaveTangent=-72.0739, tangentMode=RCTM_User, time=0, value=35}, <at 1 5 RCIM_Cubic>]}}
   }
   n122 = Get RightMovement() -> (RightMovement:double)
   n123 = Trail controller - Bone: dyn_coatFront06_l!(ComponentPose=@n121, Alpha=@n122, FakeVelocity:Vector="0,0,0") -> (Pose) {
    bActorSpaceFakeVel=true
    bShowTrailLocation=true
    chainLength=5
    debugLifeTime=0.1
    maxDeltaTime=0.0333
    relaxationSpeedScale=3
    alphaScaleBiasClamp {bInterpResult=true}
    baseJoint {boneName=root}
    planarLimits [
     {drivingBone {boneName=root}, planeTransform {location=<0 0 10>, rotation=<rot 0 0 0>, scale=<1 1 1>}}
     {drivingBone {boneName=spine_03}, planeTransform {location=<0 -20 0>, rotation=<rot 0 15 -89.9999>, scale=<1 1 1>}}
    ]
    rotationLimits [{limitMax=<180 180 180>}, {limitMax=<180 180 180>}, {limitMax=<180 180 180>}, {limitMax=<180 180 180>}, {limitMax=<180 180 180>}]
    rotationOffsets [<0 0 0>, <0 0 0>, <0 0 0>, <0 0 0>, <0 0 0>]
    trailBone {boneName=dyn_coatFront06_l}
    trailRelaxationSpeed {editorCurveData {keys [{arriveTangent=-72.0739, interpMode=RCIM_Cubic, leaveTangent=-72.0739, tangentMode=RCTM_User, time=0, value=35}, <at 1 5 RCIM_Cubic>]}}
   }
   Trail controller - Bone: dyn_coatSide05_l!(ComponentPose=@n123, Alpha=@n122, FakeVelocity:Vector="0,0,0") -> (Pose) {
    bActorSpaceFakeVel=true
    bShowTrailLocation=true
    chainLength=4
    debugLifeTime=0.1
    maxDeltaTime=0.0333
    relaxationSpeedScale=3
    alphaScaleBiasClamp {bInterpResult=true}
    baseJoint {boneName=root}
    planarLimits [
     {drivingBone {boneName=root}, planeTransform {location=<0 0 10>, rotation=<rot 0 0 0>, scale=<1 1 1>}}
     {drivingBone {boneName=spine_03}, planeTransform {location=<0 -20 0>, rotation=<rot 0 15 -89.9999>, scale=<1 1 1>}}
    ]
    rotationLimits [{limitMax=<180 180 180>}, {limitMax=<180 180 180>}, {limitMax=<180 180 180>}, {limitMax=<180 180 180>}]
    rotationOffsets [<0 0 0>, <0 0 0>, <0 0 0>, <0 0 0>]
    trailBone {boneName=dyn_coatSide05_l}
    trailRelaxationSpeed {editorCurveData {keys [{arriveTangent=-72.0739, interpMode=RCIM_Cubic, leaveTangent=-72.0739, tangentMode=RCTM_User, time=0, value=35}, <at 1 5 RCIM_Cubic>]}}
   }
  ]}
 EventGraph {nodes [
   Event BlueprintUpdateAnimation()
   > n1 = Sequence!(execute) -> (then_0, then_1, then_2:Exec, then_3:Exec, then_4:Exec)
   > Set crouchAlpha(crouchAlpha:double="To Float (Boolean)"(InBool=Get bIsCrouching())) -> (Output_Get:double)
   > Set skydive alpha(skydive alpha:double="To Float (Boolean)"(InBool=Get bIsSkyDiving())) -> (Output_Get:double)
   n2 = Get Skel_Mesh()
   > Set thigh_rot_l(thigh_rot_l=GetAlphaFromReferenceTransformDelta(SkeletalMeshComponent=@n2, BoneName=thigh_l, BaseName=pelvis, InputRange=(X=-20,Y=-115), TwistAxis="1,0,0"))
   > Set thigh_rot_r(thigh_rot_r=GetAlphaFromReferenceTransformDelta(SkeletalMeshComponent=@n2, BoneName=thigh_r, BaseName=pelvis, InputRange=(X=-20,Y=-100), TwistAxis="1,0,0"))
   n3 = GetOwnerSkeletalMeshComponent()
   n4 = BreakRotator(InRot=BreakTransform(InTransform=GetDeltaTransformFromRefPose(self=@n3, BoneName=thigh_r, BaseName=pelvis)).Rotation)
   > Set thigh_rot_up_r(thigh_rot_up_r:double=MapRangeClamped(Value=@n4.Yaw, InRangeA=-15, InRangeB=-100)) -> (Output_Get:double)
   > Set thigh_rot_inr_r(thigh_rot_inr_r:double=MapRangeClamped(Value=@n4.Pitch, InRangeA=0, InRangeB=-10, OutRangeA=1, OutRangeB=0)) -> (Output_Get:double)
   > Set thigh_rot_otr_r(thigh_rot_otr_r:double=MapRangeClamped(Value=@n4.Pitch, InRangeA=20, InRangeB=50)) -> (Output_Get:double)
   n5 = GetOwnerSkeletalMeshComponent()
   n6 = BreakRotator(InRot=BreakTransform(InTransform=GetDeltaTransformFromRefPose(self=@n5, BoneName=thigh_l, BaseName=pelvis)).Rotation)
   > Set thigh_rot_up_l(thigh_rot_up_l:double=MapRangeClamped(Value=@n6.Yaw, InRangeA=-10, InRangeB=-100)) -> (Output_Get:double)
   > Set thigh_rot_inr_l(thigh_rot_inr_l:double=MapRangeClamped(Value=@n6.Pitch, InRangeA=0, InRangeB=-10, OutRangeA=1, OutRangeB=0)) -> (Output_Get:double)
   > Set thigh_rot_otr_l(thigh_rot_otr_l:double=MapRangeClamped(Value=@n6.Pitch, InRangeA=20, InRangeB=50)) -> (Output_Get:double)
   > Set jumpingAlpha(jumpingAlpha:double="To Float (Boolean)"(InBool=Get bIsJumping())) -> (Output_Get:double)
   n7 = Get bIsMoving2D()
   > Set isNotCrouchRunning(isNotCrouchRunning:double="To Float (Boolean)"(InBool=NOT Boolean(A=@n7))) -> (Output_Get:double)
   > Set ParachuteAlpha(ParachuteAlpha:double="To Float (Boolean)"(InBool=Get bIsParachuteOpened())) -> (Output_Get:double)
   n8 = Gravity Override State Function DEPRECATED(
    execute=@n1.then_1
    CharacterPartAnimInstance=Self-Reference()
    GameTimeInSeconds=GetGameTimeInSeconds()
    GravityOverrideParams=Get Gravity Override Params()
   )
   > Set GravityOverride(GravityOverride=@n8)
   n9 = Get bIsFalling()
   n10 = Get bIsJumping()
   > n11 = Set Is jumping during sprint(Is jumping during sprint:bool=AND Boolean(A=OR Boolean(A=@n10, B=@n9), B=Get bIsTacticalSprinting())) -> (Output_Get:bool)
   > Set Coat corr(Coat corr:bool=OR Boolean(A=@n11, B=Get bIsTacticalSprinting())) -> (Output_Get:bool)
   n12 = Get bIsSprinting()
   > Set!(
    execute
    Coat rotate corr_Roll:float=0
    Coat rotate corr_Pitch:float=0
    Coat rotate corr_Yaw:float=SelectFloat(A=-3, B=SelectFloat(A=-2, B=0, bPickA=@n12), bPickA=Get Is jumping during sprint() -> (Is jumping during sprint:bool))
   ) -> (then, Output_Get:Rotator)
   n13 = Get bIsSprinting()
   > Set Left coat transtate corr(Left coat transtate corr:Vector=SelectVector(A="0.5,1.5,0", B=SelectVector(A="0,0.4,0", B="0, 0, 0", bPickA=@n13), bPickA=Get Is jumping during sprint())) -> (Output_Get:Vector)
   n14 = Get bIsSprinting()
   > Set Right coat transtate corr(Right coat transtate corr:Vector=SelectVector(A="-0.5,-1.5,0", B=SelectVector(A="0,-0.4,0", B="0, 0, 0", bPickA=@n14), bPickA=Get Is jumping during sprint())) -> (Output_Get:Vector)
   n15 = PlayerMovement(execute=@n1.then_2, AnimInstance=Self-Reference(), XDotProductCoefficient=1, YDotProductCoefficient=1, ZDotProductCoefficient=1)
   n16 = MapRangeClamped(Value=@n15.YMovement, InRangeA=0, InRangeB=-1)
   n17 = float * float(A="Absolute (Float)"(A=@n15), B=0.5)
   > Set BwdMovement(BwdMovement:double=MapRangeClamped(
     Value=float + float(A=@n17, B=@n16) {is=title K2Node_CommutativeAssociativeBinaryOperator}
     InRangeA=0
     InRangeB=1.1
     OutRangeA=0.4
     OutRangeB=0
    )) -> (Output_Get:double)
   n18 = MapRangeClamped(Value=@n15, InRangeA=-0.2, InRangeB=1)
   n19 = float * float(A="Absolute (Float)"(A=@n15.YMovement), B=0.2)
   > Set RightMovement(RightMovement:double=MapRangeClamped(Value=float + float(A=@n19, B=@n18), InRangeA=0, InRangeB=1.2, OutRangeB=0.5)) -> (Output_Get:double)
   n20 = MapRangeClamped(Value=@n15, InRangeA=0.2, InRangeB=-1, OutRangeB=0.3)
   n21 = float * float(A="Absolute (Float)"(A=@n15.YMovement), B=0.2)
   > Set LeftMovement(LeftMovement:double=MapRangeClamped(Value=float + float(A=@n21, B=@n20), InRangeA=0, InRangeB=1.2, OutRangeB=0.5)) -> (Output_Get:double)
   n22 = Branch(execute=@n1.then_3, Condition=Get bIsSkyDiving())
   > n23 = Branch(Condition="float > float"(A=Get PawnVelocityZ(), B=50) {is=title K2Node_CallFunction})
   > Set BwdMovement(BwdMovement=float * float(A=Get BwdMovement(), B=0))
   Set BwdMovement(execute=[@n22.else, @n23.else], BwdMovement=float * float(A=Get BwdMovement(), B=1))
   Set Is Sliding Float(execute=@n1.then_4, Is Sliding Float:double="To Float (Boolean)"(InBool=Get bIsSliding())) -> (Output_Get:double)
   Event BlueprintInitializeAnimation()
   > Set Skel_Mesh(Skel_Mesh=GetOwnerSkeletalMeshComponent())
   TryGetPawnOwner()
  ]}
]
variables [
 BwdMovement {type=number}
 Coat corr {type=boolean}
 Coat rotate corr {type=Rotator}
 crouchAlpha {type=number}
 Gravity Override Params {type=GravityOverrideParamsStruct}
 GravityOverride {type=Vector}
 Is jumping during sprint {type=boolean}
 Is melee or emote {type=boolean}
 Is Sliding Float {type=number}
 isNotCrouchRunning {type=number}
 jumpingAlpha {type=number}
 Left coat transtate corr {type=Vector}
 LeftMovement {type=number}
 ParachuteAlpha {type=number}
 Right coat transtate corr {type=Vector}
 RightMovement {type=number}
 Skel_Mesh {type=SkeletalMeshComponent}
 skydive alpha {type=number}
 thigh_rot_inr_l {type=number}
 thigh_rot_inr_r {type=number}
 thigh_rot_l {type=number}
 thigh_rot_otr_l {type=number}
 thigh_rot_otr_r {type=number}
 thigh_rot_r {type=number}
 thigh_rot_up_l {type=number}
 thigh_rot_up_r {type=number}
]
