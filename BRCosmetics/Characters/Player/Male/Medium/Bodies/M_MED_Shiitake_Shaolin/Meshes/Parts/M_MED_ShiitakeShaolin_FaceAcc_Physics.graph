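# Physics asset for the face accessory mesh of the Shiitake Shaolin medium male player cosmetic, covering seven dynamic cap pieces around the head.
# Each cap has a box body constrained to its own tiny base box, which is in turn constrained to a head capsule.
# Base-to-head joints rotate freely, while cap-to-base joints carry asymmetric swing limits and a very narrow twist limit, so each cap hinges mostly along one axis.
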PhysicsAsset /BRCosmetics/Characters/Player/Male/Medium/Bodies/M_MED_Shiitake_Shaolin/Meshes/Parts/M_MED_ShiitakeShaolin_FaceAcc_Physics
bodies {
 dyn_cap_1 [{shapeType=Box, extentX=4.0817, extentY=13.5425, extentZ=12.8597, center=<5 0 0>, rotation=<rot 0 90 0>}]
 dyn_cap_1_Base [{shapeType=Box, extentX=1.01, extentY=1.01, extentZ=1.01, rotation=<rot 0 -90 0>}]
 dyn_cap_2 [{shapeType=Box, extentX=5.0165, extentY=16.521, extentZ=14.6228, center=<5 0 0>, rotation=<rot 0 90 0>}]
 dyn_cap_2_Base [{shapeType=Box, extentX=1.01, extentY=1.01, extentZ=1.01, rotation=<rot 0 -90 0>}]
 dyn_cap_3 [{shapeType=Box, extentX=5.3928, extentY=14.0537, extentZ=19.3417, center=<5 0 0>, rotation=<rot 0 90 0>}]
 dyn_cap_3_Base [{shapeType=Box, extentX=1.01, extentY=1.01, extentZ=1.01, rotation=<rot 0 -90 0>}]
 dyn_cap_4 [{shapeType=Box, extentX=6.4616, extentY=16.4916, extentZ=19.1043, center=<5 0 0>, rotation=<rot 0 90 0>}]
 dyn_cap_4_Base [{shapeType=Box, extentX=1.01, extentY=1.01, extentZ=1.01, rotation=<rot 0 -90 0>}]
 dyn_cap_5 [{shapeType=Box, extentX=5.5098, extentY=14.1321, extentZ=16.5411, center=<5 0 0>, rotation=<rot 0 90 0>}]
 dyn_cap_5_Base [{shapeType=Box, extentX=1.01, extentY=1.01, extentZ=1.01, rotation=<rot 0 -90 0>}]
 dyn_cap_6 [{shapeType=Box, extentX=5.0581, extentY=15.2695, extentZ=14.1292, center=<5 0 0>, rotation=<rot 0 90 0>}]
 dyn_cap_6_Base [{shapeType=Box, extentX=1.01, extentY=1.01, extentZ=1.01, rotation=<rot 0 -90 0>}]
 dyn_cap_7 [{shapeType=Box, extentX=4.2239, extentY=11.3299, extentZ=15.4744, center=<5 0 0>, rotation=<rot 0 90 0>}]
 dyn_cap_7_Base [{shapeType=Box, extentX=1.01, extentY=1.01, extentZ=1.01, rotation=<rot 0 -90 0>}]
 head [{shapeType=Capsule, length=1.6994, radius=3.2282}]
}
constraints [
 {
  bone1Name=dyn_cap_1_Base
  bone2Name=head
  swing1LimitDegrees=45
  swing1Motion=Free
  swing2LimitDegrees=45
  swing2Motion=Free
  twistLimitDegrees=45
  twistMotion=Free
 }
 {bone1Name=dyn_cap_1, bone2Name=dyn_cap_1_Base, swing1LimitDegrees=45, swing2LimitDegrees=10, twistLimitDegrees=1}
 {
  bone1Name=dyn_cap_2_Base
  bone2Name=head
  swing1LimitDegrees=45
  swing1Motion=Free
  swing2LimitDegrees=45
  swing2Motion=Free
  twistLimitDegrees=45
  twistMotion=Free
 }
 {bone1Name=dyn_cap_2, bone2Name=dyn_cap_2_Base, swing1LimitDegrees=45, swing2LimitDegrees=10, twistLimitDegrees=1}
 {
  bone1Name=dyn_cap_3_Base
  bone2Name=head
  swing1LimitDegrees=45
  swing1Motion=Free
  swing2LimitDegrees=45
  swing2Motion=Free
  twistLimitDegrees=45
  twistMotion=Free
 }
 {bone1Name=dyn_cap_3, bone2Name=dyn_cap_3_Base, swing1LimitDegrees=45, swing2LimitDegrees=10, twistLimitDegrees=1}
 {
  bone1Name=dyn_cap_4_Base
  bone2Name=head
  swing1LimitDegrees=45
  swing1Motion=Free
  swing2LimitDegrees=45
  swing2Motion=Free
  twistLimitDegrees=45
  twistMotion=Free
 }
 {bone1Name=dyn_cap_4, bone2Name=dyn_cap_4_Base, swing1LimitDegrees=45, swing2LimitDegrees=10, twistLimitDegrees=1}
 {
  bone1Name=dyn_cap_5_Base
  bone2Name=head
  swing1LimitDegrees=45
  swing1Motion=Free
  swing2LimitDegrees=45
  swing2Motion=Free
  twistLimitDegrees=45
  twistMotion=Free
 }
 {bone1Name=dyn_cap_5, bone2Name=dyn_cap_5_Base, swing1LimitDegrees=45, swing2LimitDegrees=10, twistLimitDegrees=1}
 {
  bone1Name=dyn_cap_6_Base
  bone2Name=head
  swing1LimitDegrees=45
  swing1Motion=Free
  swing2LimitDegrees=45
  swing2Motion=Free
  twistLimitDegrees=45
  twistMotion=Free
 }
 {bone1Name=dyn_cap_6, bone2Name=dyn_cap_6_Base, swing1LimitDegrees=45, swing2LimitDegrees=10, twistLimitDegrees=1}
 {
  bone1Name=dyn_cap_7_Base
  bone2Name=head
  swing1LimitDegrees=45
  swing1Motion=Free
  swing2LimitDegrees=45
  swing2Motion=Free
  twistLimitDegrees=45
  twistMotion=Free
 }
 {bone1Name=dyn_cap_7, bone2Name=dyn_cap_7_Base, swing1LimitDegrees=45, swing2LimitDegrees=10, twistLimitDegrees=1}
]
