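# Animation Blueprint for the Shiitake Shaolin face accessory mesh, posing and simulating its hat on the player character.
# It copies the attached parent's pose, additively rotates dyn_hat_MSTR, then applies RigidBody physics at partial strength with damped component motion.
# While crouching or targeting, the head's rotation delta from spine_05 sets the hat rotation's alpha each update; otherwise that alpha is zero.
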AnimBlueprint /BRCosmetics/Characters/Player/Male/Medium/Bodies/M_MED_Shiitake_Shaolin/Meshes/Parts/M_MED_ShiitakeShaolin_FaceAcc_AnimBP
target_skeleton=&/M_MED_ShiitakeShaolin_FaceAcc_Skeleton
graphs [
 AnimGraph {nodes [
   n1 = "Transform (Modify) Bone - Bone: dyn_hat_MSTR"(
    ComponentPose=Local To Component(LocalPose=Copy Pose From Mesh() {bUseAttachedParent=true})
    Alpha=Get head rot() -> (head rot:double)
    Translation="0,0,0"
    Rotation="-15,10,0.000"
    Scale="1,1,1"
   ) {
    rotationMode=BMM_Additive
    rotationSpace=BCS_BoneSpace
    alphaBoolBlend {blendOption=Linear}
    boneToModify {boneName=dyn_hat_MSTR}
    rotation {pitch=-15, yaw=10}
   }
   n2 = RigidBody(ComponentPose=@n1, Alpha=0.6) {
    bClampLinearTranslationLimitToRefPose=true
    bOverrideWorldGravity=true
    overridePhysicsAsset=&/M_MED_ShiitakeShaolin_FaceAcc_Physics
    simulationSpace=BaseBoneSpace
    alphaBoolBlend {blendOption=Linear}
    baseBoneRef {boneName=root}
    componentAppliedLinearAccClamp=<4000 4000 4000>
    componentLinearAccScale=<0.4 0.4 0.4>
    componentLinearVelScale=<0.25 0.25 0.25>
    overrideWorldGravity {z=-980}
   }
   Output Pose(Result=Component To Local(ComponentPose=@n2)) {layerGroup=None}
  ]}
 EventGraph {nodes [
   Event BlueprintUpdateAnimation()
   > n1 = Branch(Condition=OR Boolean(A=Get bIsCrouching(), B=Get bIsTargeting()))
   > Set head rot(head rot:double=GetAlphaFromReferenceTransformDelta(SkeletalMeshComponent=Get Skel_Mesh(), BoneName=head, BaseName=spine_05, InputRange=(X=-15,Y=-40), TwistAxis="1,0,0")) -> (Output_Get:double)
   Set head rot(execute=@n1.else, head rot=0)
   Event BlueprintInitializeAnimation()
   > Set Skel_Mesh(Skel_Mesh=GetOwnerSkeletalMeshComponent())
   TryGetPawnOwner()
  ]}
]
variables [
 head rot {type=number}
 Skel_Mesh {type=SkeletalMeshComponent}
]
